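# Physics asset for the F_MED_DenimEquip_Stock female player body, built around a dangling sword keychain.
# Capsules cover the pelvis, spine, right thigh and right belt bone, all joined by angularly locked constraints, forming body volumes around the charm.
# The four-bone keychain chain hangs off spine_03 through a locked base, with its three lower joints left angularly unlocked to allow swing.
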PhysicsAsset /BRCosmetics/Characters/Player/Female/Medium/Bodies/F_MED_DenimEquip_Stock/Meshes/F_MED_DenimEquip_Stock_Physics
bodies {
 dyn_belt_r [{
   shapeType=Capsule
   length=4.2119
   radius=4.8438
   shapeName=dyn_belt_r_capsule
   center=<1.4476 -1.3782 -0.1607>
   rotation=<rot -4.484 -90 -6.919>
  }]
 pelvis [{shapeType=Capsule, length=9.7684, radius=10.855, shapeName=pelvis_capsule, center=<0 -1.2304 0>}]
 spine_01 [{shapeType=Capsule, length=0.0856, radius=0.5, shapeName=spine_01_capsule}]
 spine_02 [{shapeType=Capsule, length=14.6209, radius=8.1028, shapeName=spine_02_capsule, center=<-3.4606 0 0>}]
 spine_03 [{shapeType=Capsule, length=5.5198, radius=9.8027, shapeName=spine_03_capsule, rotation=<rot 0 -90 0>}]
 sword_keychain_01 [{
   shapeType=Capsule
   length=1.9896
   radius=1.0016
   shapeName=sword_keychain_01_capsule
   center=<2.5149 -0.1184 -0.6648>
   rotation=<rot 3.039 -90 -72.482>
  }]
 sword_keychain_02 [{
   shapeType=Capsule
   length=1.179
   radius=1.1969
   shapeName=sword_keychain_02_capsule
   center=<0.1695 0.1117 -0.0545>
   rotation=<rot 0 -83.8658 -88.5292>
  }]
 sword_keychain_03 [{
   shapeType=Capsule
   length=7.0369
   radius=1.5681
   shapeName=sword_keychain_03_capsule
   center=<3.525 0.4563 0.331>
   rotation=<rot 0 95.27 -83.5295>
  }]
 sword_keychain_base [{shapeType=Capsule, length=1.3934, radius=0.869, shapeName=sword_keychain_base_capsule, rotation=<rot 0 -90 0>}]
 thigh_r [{
   shapeType=Capsule
   length=23.9733
   radius=7.5626
   shapeName=thigh_r_capsule
   center=<26.6678 0.6242 -0.3661>
   rotation=<rot 0 -92.504 -87.5029>
  }]
}
constraints [
 {
  bone1Name=spine_02
  bone2Name=spine_01
  swing1LimitDegrees=45
  swing1Motion=Locked
  swing2LimitDegrees=45
  swing2Motion=Locked
  twistLimitDegrees=45
  twistMotion=Locked
 }
 {
  bone1Name=spine_01
  bone2Name=pelvis
  swing1LimitDegrees=45
  swing1Motion=Locked
  swing2LimitDegrees=45
  swing2Motion=Locked
  twistLimitDegrees=45
  twistMotion=Locked
 }
 {
  bone1Name=dyn_belt_r
  bone2Name=pelvis
  swing1LimitDegrees=45
  swing1Motion=Locked
  swing2LimitDegrees=45
  swing2Motion=Locked
  twistLimitDegrees=45
  twistMotion=Locked
 }
 {
  bone1Name=thigh_r
  bone2Name=pelvis
  swing1LimitDegrees=45
  swing1Motion=Locked
  swing2LimitDegrees=45
  swing2Motion=Locked
  twistLimitDegrees=45
  twistMotion=Locked
 }
 {
  bone1Name=spine_03
  bone2Name=spine_02
  swing1LimitDegrees=45
  swing1Motion=Locked
  swing2LimitDegrees=45
  swing2Motion=Locked
  twistLimitDegrees=45
  twistMotion=Locked
 }
 {
  bone1Name=sword_keychain_base
  bone2Name=spine_03
  swing1LimitDegrees=45
  swing1Motion=Locked
  swing2LimitDegrees=45
  swing2Motion=Locked
  twistLimitDegrees=45
  twistMotion=Locked
 }
 {bone1Name=sword_keychain_01, bone2Name=sword_keychain_base, swing1LimitDegrees=1, swing2LimitDegrees=5, twistLimitDegrees=1}
 {bone1Name=sword_keychain_02, bone2Name=sword_keychain_01, swing1LimitDegrees=45, swing2LimitDegrees=45, twistLimitDegrees=1}
 {bone1Name=sword_keychain_03, bone2Name=sword_keychain_02, swing1LimitDegrees=45, swing2LimitDegrees=45, twistLimitDegrees=1}
]
