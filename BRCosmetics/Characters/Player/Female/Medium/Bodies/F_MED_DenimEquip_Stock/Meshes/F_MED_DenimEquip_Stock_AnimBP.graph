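# Post-body AnimBP for the F_MED_DenimEquip_Stock female cosmetic that adds secondary motion to its boots and sword keychain.
# It copies the attached parent's pose, runs a Control Rig fed helmet and skydiving flags, then applies Anim Dynamics to boot bones.
# bRigidBodyEnabled selects RigidBody physics or an Anim Dynamics keychain chain with spherical limits.
# On customization load, it sets isHelmetOn by matching the Face part's mesh name against the mask name.
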AnimBlueprint /BRCosmetics/Characters/Player/Female/Medium/Bodies/F_MED_DenimEquip_Stock/Meshes/F_MED_DenimEquip_Stock_AnimBP
target_skeleton=&/F_MED_DenimEquip_Stock_Skeleton
graphs [
 AnimGraph {nodes [
   n1 = Anim Dynamics - Bone: sword_keychain_01(ComponentPose=Local To Component(LocalPose=Use cached pose 'AnimDynPose'()), Alpha=1, ExternalForce="0,0,0") {
    angularDampingOverride=0.9999
    angularSpringConstant=150
    bAngularSpring=true
    bOverrideAngularDamping=true
    bUseGravityOverride=true
    bUseSphericalLimits=true
    simulationSpace=RootRelative
    alphaBoolBlend {blendOption=Linear}
    boundBone {boneName=sword_keychain_01}
    componentAppliedLinearAccClamp=<2000 2000 2000>
    componentLinearAccScale=<0.3 0.3 0.3>
    componentLinearVelScale=<0.3 0.3 0.3>
    gravityOverride {z=-980}
    physicsBodyDefinitions [{
      boundBone {boneName=sword_keychain_01}
      boxExtents=<5 5 5>
      constraintSetup {angularLimitsMax=<0 60 0>, angularLimitsMin=<0 0 0>, angularTarget=<0 0 0>}
      localJointOffset=<1 0 0>
     }]
   }
   n2 = Anim Dynamics - Bone: sword_keychain_02(ComponentPose=@n1, Alpha=1, ExternalForce="0,0,0") {
    angularDampingOverride=0.9999
    angularSpringConstant=150
    bAngularSpring=true
    bOverrideAngularDamping=true
    bUseGravityOverride=true
    bUseSphericalLimits=true
    simulationSpace=RootRelative
    alphaBoolBlend {blendOption=Linear}
    boundBone {boneName=sword_keychain_02}
    componentAppliedLinearAccClamp=<2000 2000 2000>
    componentLinearAccScale=<0.3 0.3 0.3>
    componentLinearVelScale=<0.3 0.3 0.3>
    gravityOverride {z=-980}
    physicsBodyDefinitions [{
      boundBone {boneName=sword_keychain_02}
      boxExtents=<5 5 5>
      constraintSetup {angularLimitsMax=<0 10 90>, angularLimitsMin=<0 -10 0>, angularTarget=<0 0 0>}
      localJointOffset=<1 0 0>
     }]
   }
   n3 = Anim Dynamics - Bone: sword_keychain_03(ComponentPose=@n2, Alpha=1, ExternalForce="0,0,0") {
    angularDampingOverride=0.9999
    angularSpringConstant=150
    bAngularSpring=true
    bOverrideAngularDamping=true
    bUseGravityOverride=true
    bUseSphericalLimits=true
    simulationSpace=RootRelative
    alphaBoolBlend {blendOption=Linear}
    boundBone {boneName=sword_keychain_03}
    componentAppliedLinearAccClamp=<2000 2000 2000>
    componentLinearAccScale=<0.3 0.3 0.3>
    componentLinearVelScale=<0.3 0.3 0.3>
    gravityOverride {z=-980}
    physicsBodyDefinitions [{
      boundBone {boneName=sword_keychain_03}
      boxExtents=<5 5 5>
      constraintSetup {angularLimitsMax=<0 30 90>, angularLimitsMin=<0 -30 5>, angularTarget=<0 0 0>}
      localJointOffset=<1 0 0>
     }]
    sphericalLimits [
     {limitRadius=6, drivingBone {boneName=dyn_belt_r}, sphereLocalOffset=<0 -1.1277 1.2>}
     {limitRadius=12, drivingBone {boneName=spine_01}, sphereLocalOffset=<0 0 3>}
     {limitRadius=9.1918, drivingBone {boneName=thigh_r}, sphereLocalOffset=<12.6714 -0.652 0.0449>}
    ]
   }
   n4 = RigidBody(ComponentPose=Local To Component(LocalPose=Use cached pose 'AnimDynPose'())) {
    bClampLinearTranslationLimitToRefPose=true
    bOverrideWorldGravity=true
    overridePhysicsAsset=&/F_MED_DenimEquip_Stock_Physics
    simulationSpace=BaseBoneSpace
    alphaBoolBlend {blendOption=Linear}
    baseBoneRef {boneName=root}
    componentAppliedLinearAccClamp=<2000 2000 2000>
    componentLinearAccScale=<0.3 3 0.3>
    componentLinearVelScale=<0.3 0.3 0.3>
    overrideWorldGravity {z=-980}
   }
   Output Pose(Result=Blend Poses by bool(BlendPose_0=Component To Local(ComponentPose=@n4), BlendPose_1=Component To Local(ComponentPose=@n3), bActiveValue=Get bRigidBodyEnabled()) {blendType=Linear}) {layerGroup=None}
   n5 = Control Rig!(
    Source=Copy Pose From Mesh() {bUseAttachedParent=true}
    Alpha
    isHelmetOn:bool=Get isHelmetOn() -> (isHelmetOn:bool)
    isSkyDiving:bool=Get bIsSkyDiving()
   ) -> (Pose) {
    alphaBoolBlend {blendOption=Linear}
    controlRigAssetReference {blueprintRigClass=&/F_MED_DenimEquip_Stock_CtrlRig2.F_MED_DenimEquip_Stock_CtrlRig2_C}
   }
   Save cached pose 'StartPose'(Pose=@n5) {cacheName=StartPose}
   n6 = Anim Dynamics - Bone: dyn_boot_front_l(ComponentPose=Local To Component(LocalPose=Use cached pose 'StartPose'()), Alpha=1, ExternalForce="0,0,0") {
    angularDampingOverride=0.9
    angularSpringConstant=2000
    bAngularSpring=true
    bGravityOverrideInSimSpace=true
    bOverrideAngularDamping=true
    bUseGravityOverride=true
    simulationSpace=RootRelative
    alphaBoolBlend {blendOption=Linear}
    boundBone {boneName=dyn_boot_front_l}
    componentAppliedLinearAccClamp=<3000 3000 3000>
    componentLinearAccScale=<0.3 0.3 0.3>
    componentLinearVelScale=<0.3 0.3 0.3>
    gravityOverride {z=-980}
    physicsBodyDefinitions [{
      boundBone {boneName=dyn_boot_front_l}
      boxExtents=<15 15 15>
      constraintSetup {angularLimitsMax=<5 20 20>, angularLimitsMin=<-5 -40 -20>, angularTarget=<0 0 0>}
      localJointOffset=<1 0 0>
     }]
   }
   n7 = Anim Dynamics - Bone: dyn_boot_back_l(ComponentPose=@n6, Alpha=1, ExternalForce="0,0,0") {
    angularDampingOverride=0.9
    angularSpringConstant=1500
    bAngularSpring=true
    bGravityOverrideInSimSpace=true
    bOverrideAngularDamping=true
    bUseGravityOverride=true
    simulationSpace=RootRelative
    alphaBoolBlend {blendOption=Linear}
    boundBone {boneName=dyn_boot_back_l}
    componentAppliedLinearAccClamp=<3000 3000 3000>
    componentLinearAccScale=<0.3 0.3 0.3>
    componentLinearVelScale=<0.3 0.3 0.3>
    gravityOverride {z=-980}
    physicsBodyDefinitions [{
      boundBone {boneName=dyn_boot_back_l}
      boxExtents=<10 10 10>
      constraintSetup {angularLimitsMax=<5 30 20>, angularLimitsMin=<-5 -40 -20>, angularTarget=<0 0 0>}
      localJointOffset=<1.5 0 -0.2>
     }]
   }
   n8 = Anim Dynamics - Bone: dyn_boot_side01_l(ComponentPose=@n7, Alpha=1, ExternalForce="0,0,0") {
    angularDampingOverride=0.9
    angularSpringConstant=1000
    bAngularSpring=true
    bGravityOverrideInSimSpace=true
    bOverrideAngularDamping=true
    bUseGravityOverride=true
    simulationSpace=RootRelative
    alphaBoolBlend {blendOption=Linear}
    boundBone {boneName=dyn_boot_side01_l}
    componentAppliedLinearAccClamp=<3000 3000 3000>
    componentLinearAccScale=<0.3 0.3 0.3>
    componentLinearVelScale=<0.3 0.3 0.3>
    gravityOverride {z=-980}
    physicsBodyDefinitions [{
      boundBone {boneName=dyn_boot_side01_l}
      boxExtents=<10 15 10>
      constraintSetup {angularLimitsMax=<5 20 5>, angularLimitsMin=<-5 -40 -5>, angularTarget=<0 0 0>}
      localJointOffset=<1 0 -0.5>
     }]
   }
   n9 = Anim Dynamics - Bone: dyn_boot_side02_l(ComponentPose=@n8, Alpha=1, ExternalForce="0,0,0") {
    angularDampingOverride=0.9
    angularSpringConstant=1000
    bAngularSpring=true
    bGravityOverrideInSimSpace=true
    bOverrideAngularDamping=true
    bUseGravityOverride=true
    simulationSpace=RootRelative
    alphaBoolBlend {blendOption=Linear}
    boundBone {boneName=dyn_boot_side02_l}
    componentAppliedLinearAccClamp=<3000 3000 3000>
    componentLinearAccScale=<0.3 0.3 0.3>
    componentLinearVelScale=<0.3 0.3 0.3>
    gravityOverride {z=-980}
    physicsBodyDefinitions [{
      boundBone {boneName=dyn_boot_side02_l}
      boxExtents=<10 15 10>
      constraintSetup {angularLimitsMax=<5 20 5>, angularLimitsMin=<-5 -40 -5>, angularTarget=<0 0 0>}
      localJointOffset=<1 0 -0.5>
     }]
   }
   n10 = Anim Dynamics - Bone: dyn_boot_front_r(ComponentPose=@n9, Alpha=1, ExternalForce="0,0,0") {
    angularDampingOverride=0.9
    angularSpringConstant=2000
    bAngularSpring=true
    bGravityOverrideInSimSpace=true
    bOverrideAngularDamping=true
    bUseGravityOverride=true
    simulationSpace=RootRelative
    alphaBoolBlend {blendOption=Linear}
    boundBone {boneName=dyn_boot_front_r}
    componentAppliedLinearAccClamp=<3000 3000 3000>
    componentLinearAccScale=<0.3 0.3 0.3>
    componentLinearVelScale=<0.3 0.3 0.3>
    gravityOverride {z=-980}
    physicsBodyDefinitions [{
      boundBone {boneName=dyn_boot_front_r}
      boxExtents=<15 15 15>
      constraintSetup {angularLimitsMax=<5 20 20>, angularLimitsMin=<-5 -40 -20>, angularTarget=<0 0 0>}
      localJointOffset=<1 0 0>
     }]
   }
   n11 = Anim Dynamics - Bone: dyn_boot_back_r(ComponentPose=@n10, Alpha=1, ExternalForce="0,0,0") {
    angularDampingOverride=0.9
    angularSpringConstant=1500
    bAngularSpring=true
    bGravityOverrideInSimSpace=true
    bOverrideAngularDamping=true
    bUseGravityOverride=true
    simulationSpace=RootRelative
    alphaBoolBlend {blendOption=Linear}
    boundBone {boneName=dyn_boot_back_r}
    componentAppliedLinearAccClamp=<3000 3000 3000>
    componentLinearAccScale=<0.3 0.3 0.3>
    componentLinearVelScale=<0.3 0.3 0.3>
    gravityOverride {z=-980}
    physicsBodyDefinitions [{
      boundBone {boneName=dyn_boot_back_r}
      boxExtents=<10 10 10>
      constraintSetup {angularLimitsMax=<5 30 20>, angularLimitsMin=<-5 -40 -20>, angularTarget=<0 0 0>}
      localJointOffset=<1.5 0 -0.2>
     }]
   }
   Save cached pose 'AnimDynPose'(Pose=Component To Local(ComponentPose=@n11)) {cacheName=AnimDynPose}
  ]}
 EventGraph {nodes [
   n1 = Event OnCharacterCustomizationLoaded()
   n2 = GetSkinnedAsset!(self:Skinned Mesh Component Object Reference=Get Array Item with EFort Custom Part Type(Array=@n1.CustomMeshParts, PartType=Face)) -> (ReturnValue:Skinned Asset Object Reference)
   > Set isHelmetOn(isHelmetOn:bool="Equal Exactly (String)"(A=GetDisplayName(Object=@n2), B=Get isHelmet() -> (isHelmet:String))) -> (Output_Get:bool)
   Event BlueprintUpdateAnimation()
  ]}
]
variables [
 isHelmet {type=string, default=F_MED_DenimEquip_StockMask_FaceAcc}
 isHelmetOn {type=boolean}
]
