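# Physics asset for the FaceAcc mesh part of the female medium DenimEquip Stock body, defining bodies and joints for dynamic hair.
# Five hair chains (center, two left, two right) each link two capsule segments to a small master-bone capsule, with narrow swing limits and zero twist limits.
# Larger unconstrained capsules on the head, hair master, neck, upper spine and right upper arm give body shapes around the hair, with no left-arm counterpart.
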PhysicsAsset /BRCosmetics/Characters/Player/Female/Medium/Bodies/F_MED_DenimEquip_Stock/Meshes/Parts/F_MED_DenimEquip_Stock_FaceAcc_Physics
bodies {
 dyn_hair_MSTR_c_01 [{shapeType=Capsule, length=1, radius=1, shapeName=dyn_hair_MSTR_r_01_capsule}]
 dyn_hair_MSTR_l2_01 [{shapeType=Capsule, length=1, radius=1, shapeName=dyn_hair_MSTR_r_01_capsule}]
 dyn_hair_MSTR_l_01 [{shapeType=Capsule, length=1, radius=1, shapeName=dyn_hair_MSTR_r_01_capsule}]
 dyn_hair_MSTR_r2_01 [{shapeType=Capsule, length=1, radius=1, shapeName=dyn_hair_MSTR_r_01_capsule}]
 dyn_hair_MSTR_r_01 [{shapeType=Capsule, length=1, radius=1, shapeName=dyn_hair_MSTR_r_01_capsule}]
 dyn_hair_c_01 [{shapeType=Capsule, length=6, radius=2, shapeName=dyn_hair_r_01_capsule, center=<5.6196 -0.0001 0>, rotation=<rot -90 0.0001 -0.0009>}]
 dyn_hair_c_02 [{shapeType=Capsule, length=10, radius=3, shapeName=dyn_hair_r_01_capsule, center=<6.7889 -0.0001 0>, rotation=<rot -90 0.0001 -0.0009>}]
 dyn_hair_l2_01 [{shapeType=Capsule, length=6, radius=2, shapeName=dyn_hair_r_01_capsule, center=<5.6196 -0.0001 0>, rotation=<rot -90 0.0001 -0.0009>}]
 dyn_hair_l2_02 [{shapeType=Capsule, length=7, radius=4, shapeName=dyn_hair_r_02_capsule, center=<0 2.5713 0>, rotation=<rot 0 -0.0002 -90>}]
 dyn_hair_l_01 [{shapeType=Capsule, length=6, radius=2, shapeName=dyn_hair_r_01_capsule, center=<5.6196 -0.0001 0>, rotation=<rot -90 0.0001 -0.0009>}]
 dyn_hair_l_02 [{shapeType=Capsule, length=7, radius=4, shapeName=dyn_hair_r_02_capsule, rotation=<rot 0 -0.0002 -90>}]
 dyn_hair_r2_01 [{shapeType=Capsule, length=6, radius=2, shapeName=dyn_hair_r_01_capsule, center=<5.6196 -0.0001 0>, rotation=<rot -90 0.0001 -0.0009>}]
 dyn_hair_r2_02 [{shapeType=Capsule, length=7, radius=2, shapeName=dyn_hair_r_02_capsule, center=<0 1.4873 0>, rotation=<rot 0 -0.0002 -90>}]
 dyn_hair_r_01 [{shapeType=Capsule, length=6, radius=2, shapeName=dyn_hair_r_01_capsule, center=<5.6196 -0.0001 0>, rotation=<rot -90 0.0001 -0.0009>}]
 dyn_hair_r_02 [{shapeType=Capsule, length=7, radius=4, shapeName=dyn_hair_r_02_capsule, rotation=<rot 0 -0.0002 -90>}]
 hair_MSTR [
  {shapeType=Capsule, length=6.5999, radius=6.2529, shapeName=head_capsule, center=<3.4929 3.5073 0>, rotation=<rot -90 0.0003 -0.0004>}
  {shapeType=Capsule, length=3.1156, radius=5.8885, shapeName=head_capsule, center=<6.8802 1.9038 0>, rotation=<rot 0 -10.0004 -90>}
 ]
 head [
  {shapeType=Capsule, length=6.5999, radius=6.2529, shapeName=head_capsule, center=<3.4929 3.5073 0>, rotation=<rot -90 0.0003 -0.0004>}
  {shapeType=Capsule, length=3.1156, radius=5.8885, shapeName=head_capsule, center=<6.8802 1.9038 0>, rotation=<rot 0 -10.0004 -90>}
 ]
 neck_01 [{shapeType=Capsule, length=5.9894, radius=4.714, shapeName=neck_01_capsule, rotation=<rot -90 180 179.9999>}]
 spine_05 [{shapeType=Capsule, length=15, radius=9.5, shapeName=spine_05_capsule, center=<3.8892 -0.3873 0>}]
 upperarm_r [{shapeType=Capsule, length=15, radius=4, shapeName=upperarm_r_capsule, center=<-12.9529 0 2.0478>, rotation=<rot -90 180 180>}]
}
constraints [
 {bone1Name=dyn_hair_r_01, bone2Name=dyn_hair_MSTR_r_01, swing1LimitDegrees=5, swing2LimitDegrees=10, twistLimitDegrees=0}
 {bone1Name=dyn_hair_r_02, bone2Name=dyn_hair_r_01, swing1LimitDegrees=5, swing2LimitDegrees=10, twistLimitDegrees=0}
 {bone1Name=dyn_hair_r2_01, bone2Name=dyn_hair_MSTR_r2_01, swing1LimitDegrees=15, swing2LimitDegrees=20, twistLimitDegrees=0}
 {bone1Name=dyn_hair_r2_02, bone2Name=dyn_hair_r2_01, swing1LimitDegrees=15, swing2LimitDegrees=10, twistLimitDegrees=0}
 {bone1Name=dyn_hair_c_01, bone2Name=dyn_hair_MSTR_c_01, swing1LimitDegrees=15, swing2LimitDegrees=20, twistLimitDegrees=0}
 {bone1Name=dyn_hair_c_02, bone2Name=dyn_hair_c_01, swing1LimitDegrees=15, swing2LimitDegrees=10, twistLimitDegrees=0}
 {bone1Name=dyn_hair_l_01, bone2Name=dyn_hair_MSTR_l_01, swing1LimitDegrees=5, swing2LimitDegrees=10, twistLimitDegrees=0}
 {bone1Name=dyn_hair_l_02, bone2Name=dyn_hair_l_01, swing1LimitDegrees=5, swing2LimitDegrees=10, twistLimitDegrees=0}
 {bone1Name=dyn_hair_l2_01, bone2Name=dyn_hair_MSTR_l2_01, swing1LimitDegrees=15, swing2LimitDegrees=20, twistLimitDegrees=0}
 {bone1Name=dyn_hair_l2_02, bone2Name=dyn_hair_l2_01, swing1LimitDegrees=15, swing2LimitDegrees=10, twistLimitDegrees=0}
]
