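# Animation Blueprint for the FaceAcc part of the F_MED_DenimEquip_Stock female cosmetic body, adding secondary motion and corrective posing to its hair bones.
# AnimDynamics nodes simulate bangs, tails, back hair and the master hair bone under custom gravity.
# Additive bone transforms, blended through cached poses, offset strands by head rotation, crouch, targeting, jump/fall and skydive state.
# A Control Rig pass applies during melee animations, and some node chains lack a pose input.
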AnimBlueprint /BRCosmetics/Characters/Player/Female/Medium/Bodies/F_MED_DenimEquip_Stock/Meshes/Parts/F_MED_DenimEquip_Stock_FaceAcc_AnimBP
target_skeleton=&/F_MED_DenimEquip_Stock_FaceAcc_Skeleton
graphs [
 AnimGraph {nodes [
   Anim Dynamics - Bone: dyn_hairBack_01_c!(ComponentPose, bAlphaBoolEnabled:bool=True, ExternalForce="0,0,0") -> (Pose) {
    alphaInputType=Bool
    angularDampingOverride=1
    angularSpringConstant=20000
    bOverrideAngularDamping=true
    bUseGravityOverride=true
    simulationSpace=RootRelative
    alphaBoolBlend {blendInTime=0.1, blendOption=Linear, blendOutTime=0.2}
    boundBone {boneName=dyn_hairBack_01_c}
    componentAppliedLinearAccClamp=<100 100 100>
    componentLinearAccScale=<0.2 0.2 0.2>
    componentLinearVelScale=<0.3 0.3 0.3>
    gravityOverride {z=-900}
    physicsBodyDefinitions [{
      boundBone {boneName=dyn_hairBack_01_c}
      boxExtents=<12 12 12>
      constraintSetup {angularLimitsMax=<4 5 0>, angularLimitsMin=<-4 -5 -25>, angularTarget=<0 0 0>}
      localJointOffset=<1 0 0>
     }]
   }
   n1 = Anim Dynamics - Bone: dyn_hair_tails_01_r!(ComponentPose, bAlphaBoolEnabled:bool=True, ExternalForce="0,0,0") -> (Pose) {
    alphaInputType=Bool
    angularDampingOverride=1
    bOverrideAngularDamping=true
    bUseGravityOverride=true
    simulationSpace=Actor
    alphaBoolBlend {blendInTime=0.1, blendOption=Linear, blendOutTime=0.2}
    boundBone {boneName=dyn_hair_tails_01_r}
    componentAppliedLinearAccClamp=<2000 2000 2000>
    componentLinearAccScale=<0.2 0.2 0.2>
    componentLinearVelScale=<0.3 0.3 0.3>
    gravityOverride {z=-700}
    physicsBodyDefinitions [{
      boundBone {boneName=dyn_hair_tails_01_r}
      boxExtents=<5 5 5>
      constraintSetup {angularLimitsMax=<5 30 20>, angularLimitsMin=<-5 0 -20>, angularTarget=<0 0 0>}
      localJointOffset=<1 0 0>
     }]
   }
   n2 = Anim Dynamics - Bone: dyn_hair_tails_02_r!(ComponentPose=@n1, bAlphaBoolEnabled:bool=True, ExternalForce="0,0,0") -> (Pose) {
    alphaInputType=Bool
    angularDampingOverride=1
    bOverrideAngularDamping=true
    bUseGravityOverride=true
    simulationSpace=Actor
    alphaBoolBlend {blendInTime=0.1, blendOption=Linear, blendOutTime=0.2}
    boundBone {boneName=dyn_hair_tails_02_r}
    componentAppliedLinearAccClamp=<2000 2000 2000>
    componentLinearAccScale=<0.2 0.2 0.2>
    componentLinearVelScale=<0.3 0.3 0.3>
    gravityOverride {z=-700}
    physicsBodyDefinitions [{
      boundBone {boneName=dyn_hair_tails_02_r}
      boxExtents=<5 5 5>
      constraintSetup {angularLimitsMax=<5 30 20>, angularLimitsMin=<-5 0 -20>, angularTarget=<0 0 0>}
      localJointOffset=<1 0 0>
     }]
   }
   n3 = Anim Dynamics - Bone: dyn_hair_tails_03_r!(ComponentPose=@n2, bAlphaBoolEnabled:bool=True, ExternalForce="0,0,0") -> (Pose) {
    alphaInputType=Bool
    angularDampingOverride=1
    bOverrideAngularDamping=true
    bUseGravityOverride=true
    simulationSpace=Actor
    alphaBoolBlend {blendInTime=0.1, blendOption=Linear, blendOutTime=0.2}
    boundBone {boneName=dyn_hair_tails_03_r}
    componentAppliedLinearAccClamp=<2000 2000 2000>
    componentLinearAccScale=<0.2 0.2 0.2>
    componentLinearVelScale=<0.3 0.3 0.3>
    gravityOverride {z=-700}
    physicsBodyDefinitions [{
      boundBone {boneName=dyn_hair_tails_03_r}
      boxExtents=<5 5 5>
      constraintSetup {angularLimitsMax=<5 30 20>, angularLimitsMin=<-5 0 -20>, angularTarget=<0 0 0>}
      localJointOffset=<1 0 0>
     }]
   }
   n4 = Anim Dynamics - Bone: dyn_hair_bang_03_r!(ComponentPose=@n3, bAlphaBoolEnabled:bool=True, ExternalForce="0,0,0") -> (Pose) {
    alphaInputType=Bool
    angularDampingOverride=1
    bOverrideAngularDamping=true
    bUseGravityOverride=true
    simulationSpace=Actor
    alphaBoolBlend {blendInTime=0.1, blendOption=Linear, blendOutTime=0.2}
    boundBone {boneName=dyn_hair_bang_03_r}
    componentAppliedLinearAccClamp=<2000 2000 2000>
    componentLinearAccScale=<0.2 0.2 0.2>
    componentLinearVelScale=<0.3 0.3 0.3>
    gravityOverride {z=-700}
    physicsBodyDefinitions [{
      boundBone {boneName=dyn_hair_bang_03_r}
      boxExtents=<5 5 5>
      constraintSetup {angularLimitsMax=<5 30 20>, angularLimitsMin=<-5 0 -20>, angularTarget=<0 0 0>}
      localJointOffset=<1 0 0>
     }]
   }
   n5 = Anim Dynamics - Bone: dyn_hair_tails_01_l!(ComponentPose=@n4, bAlphaBoolEnabled:bool=True, ExternalForce="0,0,0") -> (Pose) {
    alphaInputType=Bool
    angularDampingOverride=1
    bOverrideAngularDamping=true
    bUseGravityOverride=true
    simulationSpace=Actor
    alphaBoolBlend {blendInTime=0.1, blendOption=Linear, blendOutTime=0.2}
    boundBone {boneName=dyn_hair_tails_01_l}
    componentAppliedLinearAccClamp=<2000 2000 2000>
    componentLinearAccScale=<0.2 0.2 0.2>
    componentLinearVelScale=<0.3 0.3 0.3>
    gravityOverride {z=-700}
    physicsBodyDefinitions [{
      boundBone {boneName=dyn_hair_tails_01_l}
      boxExtents=<5 5 5>
      constraintSetup {angularLimitsMax=<5 30 20>, angularLimitsMin=<-5 0 -20>, angularTarget=<0 0 0>}
      localJointOffset=<1 0 0>
     }]
   }
   n6 = Anim Dynamics - Bone: dyn_hair_tails_02_l!(ComponentPose=@n5, bAlphaBoolEnabled:bool=True, ExternalForce="0,0,0") -> (Pose) {
    alphaInputType=Bool
    angularDampingOverride=1
    bOverrideAngularDamping=true
    bUseGravityOverride=true
    simulationSpace=Actor
    alphaBoolBlend {blendInTime=0.1, blendOption=Linear, blendOutTime=0.2}
    boundBone {boneName=dyn_hair_tails_02_l}
    componentAppliedLinearAccClamp=<2000 2000 2000>
    componentLinearAccScale=<0.2 0.2 0.2>
    componentLinearVelScale=<0.3 0.3 0.3>
    gravityOverride {z=-700}
    physicsBodyDefinitions [{
      boundBone {boneName=dyn_hair_tails_02_l}
      boxExtents=<5 5 5>
      constraintSetup {angularLimitsMax=<5 30 20>, angularLimitsMin=<-5 0 -20>, angularTarget=<0 0 0>}
      localJointOffset=<1 0 0>
     }]
   }
   n7 = Anim Dynamics - Bone: dyn_hair_tails_03_l!(ComponentPose=@n6, bAlphaBoolEnabled:bool=True, ExternalForce="0,0,0") -> (Pose) {
    alphaInputType=Bool
    angularDampingOverride=1
    bOverrideAngularDamping=true
    bUseGravityOverride=true
    simulationSpace=Actor
    alphaBoolBlend {blendInTime=0.1, blendOption=Linear, blendOutTime=0.2}
    boundBone {boneName=dyn_hair_tails_03_l}
    componentAppliedLinearAccClamp=<2000 2000 2000>
    componentLinearAccScale=<0.2 0.2 0.2>
    componentLinearVelScale=<0.3 0.3 0.3>
    gravityOverride {z=-700}
    physicsBodyDefinitions [{
      boundBone {boneName=dyn_hair_tails_03_l}
      boxExtents=<5 5 5>
      constraintSetup {angularLimitsMax=<5 30 20>, angularLimitsMin=<-5 0 -20>, angularTarget=<0 0 0>}
      localJointOffset=<1 0 0>
     }]
   }
   n8 = Anim Dynamics - Bone: dyn_hair_tails_04_l!(ComponentPose=@n7, bAlphaBoolEnabled:bool=True, ExternalForce="0,0,0") -> (Pose) {
    alphaInputType=Bool
    angularDampingOverride=1
    bOverrideAngularDamping=true
    bUseGravityOverride=true
    simulationSpace=Actor
    alphaBoolBlend {blendInTime=0.1, blendOption=Linear, blendOutTime=0.2}
    boundBone {boneName=dyn_hair_tails_04_l}
    componentAppliedLinearAccClamp=<2000 2000 2000>
    componentLinearAccScale=<0.2 0.2 0.2>
    componentLinearVelScale=<0.3 0.3 0.3>
    gravityOverride {z=-700}
    physicsBodyDefinitions [{
      boundBone {boneName=dyn_hair_tails_04_l}
      boxExtents=<5 5 5>
      constraintSetup {angularLimitsMax=<5 30 20>, angularLimitsMin=<-5 0 -20>, angularTarget=<0 0 0>}
      localJointOffset=<1 0 0>
     }]
   }
   Anim Dynamics - Bone: dyn_hair_tails_05_l!(ComponentPose=@n8, bAlphaBoolEnabled:bool=True, ExternalForce="0,0,0") -> (Pose) {
    alphaInputType=Bool
    angularDampingOverride=1
    bOverrideAngularDamping=true
    bUseGravityOverride=true
    simulationSpace=Actor
    alphaBoolBlend {blendInTime=0.1, blendOption=Linear, blendOutTime=0.2}
    boundBone {boneName=dyn_hair_tails_05_l}
    componentAppliedLinearAccClamp=<2000 2000 2000>
    componentLinearAccScale=<0.2 0.2 0.2>
    componentLinearVelScale=<0.3 0.3 0.3>
    gravityOverride {z=-700}
    physicsBodyDefinitions [{
      boundBone {boneName=dyn_hair_tails_05_l}
      boxExtents=<5 5 5>
      constraintSetup {angularLimitsMax=<5 30 20>, angularLimitsMin=<-5 0 -20>, angularTarget=<0 0 0>}
      localJointOffset=<1 0 0>
     }]
   }
   Anim Dynamics - Bone: hair_MSTR(Alpha=0.25, ExternalForce="0,0,0") {
    angularSpringConstant=80
    bAngularSpring=true
    bGravityOverrideInSimSpace=true
    bLinearSpring=true
    lODThreshold=1
    linearSpringConstant=80
    simulationSpace=Actor
    alphaBoolBlend {blendOption=Linear}
    boundBone {boneName=hair_MSTR}
    physicsBodyDefinitions [{
      sphereCollisionRadius=0
      boundBone {boneName=hair_MSTR}
      boxExtents=<30 30 30>
      constraintSetup {
       angularTargetAxis=AxisY
       angularLimitsMax=<0 30 30>
       angularLimitsMin=<0 -30 -30>
       angularTarget=<0 0 0>
       linearAxesMax=<0 5 5>
       linearAxesMin=<-5 -5 -5>
      }
      localJointOffset=<0 0 0>
     }]
   }
   Anim Dynamics - Bone: hair_MSTR!(ComponentPose, bAlphaBoolEnabled:bool=True, ExternalForce="0,0,0") -> (Pose) {
    alphaInputType=Bool
    angularSpringConstant=80
    bGravityOverrideInSimSpace=true
    bUseGravityOverride=true
    lODThreshold=1
    linearDampingOverride=0.98
    linearSpringConstant=50
    simulationSpace=BoneRelative
    alphaBoolBlend {blendInTime=0.1, blendOption=Linear, blendOutTime=0.8}
    boundBone {boneName=hair_MSTR}
    componentAppliedLinearAccClamp=<600 600 600>
    componentLinearAccScale=<0.2 0.2 0.35>
    componentLinearVelScale=<0.3 0.3 0.45>
    gravityOverride {z=-400}
    physicsBodyDefinitions [{
      sphereCollisionRadius=0
      boundBone {boneName=hair_MSTR}
      boxExtents=<40 40 40>
      constraintSetup {angularTargetAxis=AxisY, angularLimitsMax=<0 0 0>, angularLimitsMin=<0 0 0>, angularTarget=<0 0 0>, linearAxesMin=<-4 0 0>}
      localJointOffset=<0 0 0>
     }]
   }
   Anim Dynamics - Bone: dyn_hair_c_02!(ComponentPose, bAlphaBoolEnabled:bool=True, ExternalForce="0,0,0") -> (Pose) {
    alphaInputType=Bool
    angularDampingOverride=1
    bOverrideAngularDamping=true
    bUseGravityOverride=true
    bUsePlanarLimit=false
    simulationSpace=RootRelative
    alphaBoolBlend {blendInTime=0.1, blendOption=Linear, blendOutTime=0.2}
    boundBone {boneName=dyn_hair_c_02}
    componentAppliedLinearAccClamp=<2000 2000 2000>
    componentLinearAccScale=<0.2 0.2 0.2>
    componentLinearVelScale=<0.3 0.3 0.3>
    gravityOverride {z=-700}
    physicsBodyDefinitions [{
      boundBone {boneName=dyn_hair_c_02}
      boxExtents=<10 10 10>
      constraintSetup {angularLimitsMax=<5 0 0>, angularLimitsMin=<-5 -20 -5>, angularTarget=<0 0 0>}
      localJointOffset=<1 0 0>
     }]
   }
   Anim Dynamics - Bone: hair_MSTR(bAlphaBoolEnabled=Get bIsJumping(), ExternalForce="0,0,0") {
    alphaInputType=Bool
    angularSpringConstant=80
    bGravityOverrideInSimSpace=true
    bUseGravityOverride=true
    lODThreshold=1
    linearDampingOverride=0.98
    linearSpringConstant=50
    simulationSpace=RootRelative
    alphaBoolBlend {blendInTime=0.1, blendOption=Linear, blendOutTime=0.5}
    boundBone {boneName=hair_MSTR}
    componentAppliedLinearAccClamp=<600 600 600>
    componentLinearAccScale {z=1}
    componentLinearVelScale {z=1}
    gravityOverride {z=-980}
    physicsBodyDefinitions [{
      sphereCollisionRadius=0
      boundBone {boneName=hair_MSTR}
      boxExtents=<40 40 40>
      constraintSetup {angularTargetAxis=AxisY, angularLimitsMax=<0 0 0>, angularLimitsMin=<0 0 0>, angularTarget=<0 0 0>, linearAxesMin=<-4 0 0>}
      localJointOffset=<0 0 0>
     }]
   }
   n9 = Get Head_rotation_lowAlpha() -> (Head_rotation_lowAlpha:double)
   n10 = "Transform (Modify) Bone - Bone: dyn_hair_r_02"(ComponentPose=Local To Component(), Alpha=@n9, Translation="-2,0,0", Rotation="10,10,0", Scale="1,1,1") {
    rotationMode=BMM_Additive
    rotationSpace=BCS_BoneSpace
    scaleMode=BMM_Additive
    scaleSpace=BCS_BoneSpace
    translationMode=BMM_Additive
    translationSpace=BCS_BoneSpace
    alphaBoolBlend {blendOption=Linear}
    boneToModify {boneName=dyn_hair_r_02}
    rotation {pitch=10, yaw=10}
    translation {x=-2}
   }
   n11 = "Transform (Modify) Bone - Bone: dyn_hair_c_02"(ComponentPose=@n10, Alpha=@n9, Translation="-1,0,0", Rotation="10,0,0", Scale="1,1,1") {
    rotationMode=BMM_Additive
    rotationSpace=BCS_BoneSpace
    scaleMode=BMM_Additive
    scaleSpace=BCS_BoneSpace
    translationMode=BMM_Additive
    translationSpace=BCS_BoneSpace
    alphaBoolBlend {blendOption=Linear}
    boneToModify {boneName=dyn_hair_c_02}
    rotation {pitch=10}
    translation {x=-1}
   }
   n12 = "Transform (Modify) Bone - Bone: dyn_hair_c_02"(ComponentPose=@n11, Alpha=Get Head_rotation() -> (Head_rotation:double), Translation="-3,0,0", Rotation="0,0,0", Scale="1,1,1") {
    rotationMode=BMM_Additive
    rotationSpace=BCS_BoneSpace
    scaleMode=BMM_Additive
    scaleSpace=BCS_BoneSpace
    translationMode=BMM_Additive
    translationSpace=BCS_BoneSpace
    alphaBoolBlend {blendInTime=0.1, blendOption=Linear, blendOutTime=0.1}
    boneToModify {boneName=dyn_hair_c_02}
    translation {x=-3}
   }
   Blend Poses by bool(
    BlendPose_0=Component To Local(ComponentPose=@n12)
    BlendPose_1=Use cached pose 'StandLogic'()
    bActiveValue=Get CrouchTargeting() -> (CrouchTargeting:bool)
   ) {blendType=Linear}
   n13 = Anim Dynamics - Bone: dyn_hair_bang_01!(ComponentPose=Local To Component(), bAlphaBoolEnabled:bool=True, ExternalForce="0,0,0") -> (Pose) {
    alphaInputType=Bool
    angularDampingOverride=1
    bOverrideAngularDamping=true
    bUseGravityOverride=true
    simulationSpace=RootRelative
    alphaBoolBlend {blendInTime=0.1, blendOption=Linear, blendOutTime=0.2}
    boundBone {boneName=dyn_hair_bang_01}
    componentAppliedLinearAccClamp=<2000 2000 2000>
    componentLinearAccScale=<0.2 0.2 0.2>
    componentLinearVelScale=<0.3 0.3 0.3>
    gravityOverride {z=-700}
    physicsBodyDefinitions [{
      boundBone {boneName=dyn_hair_bang_01}
      boxExtents=<10 10 10>
      constraintSetup {angularLimitsMax=<5 20 0>, angularLimitsMin=<-5 -20 -5>, angularTarget=<0 0 0>}
      localJointOffset=<1 0 0>
     }]
   }
   n14 = Anim Dynamics - Bone: dyn_hair_bang_02!(ComponentPose=@n13, bAlphaBoolEnabled:bool=True, ExternalForce="0,0,0") -> (Pose) {
    alphaInputType=Bool
    angularDampingOverride=1
    bOverrideAngularDamping=true
    bUseGravityOverride=true
    simulationSpace=RootRelative
    alphaBoolBlend {blendInTime=0.1, blendOption=Linear, blendOutTime=0.2}
    boundBone {boneName=dyn_hair_bang_02}
    componentAppliedLinearAccClamp=<2000 2000 2000>
    componentLinearAccScale=<0.2 0.2 0.2>
    componentLinearVelScale=<0.3 0.3 0.3>
    gravityOverride {z=-700}
    physicsBodyDefinitions [{
      boundBone {boneName=dyn_hair_bang_02}
      boxExtents=<10 10 10>
      constraintSetup {angularLimitsMax=<5 20 0>, angularLimitsMin=<-5 -20 -5>, angularTarget=<0 0 0>}
      localJointOffset=<1 0 0>
     }]
   }
   n15 = Get bIsJumping()
   n16 = Anim Dynamics - Bone: hair_MSTR(ComponentPose=@n14, bAlphaBoolEnabled=@n15, ExternalForce="0,0,0") {
    alphaInputType=Bool
    angularSpringConstant=80
    bGravityOverrideInSimSpace=true
    bOverrideLinearDamping=true
    bUseGravityOverride=true
    lODThreshold=1
    linearDampingOverride=0.98
    linearSpringConstant=50
    simulationSpace=RootRelative
    alphaBoolBlend {blendInTime=0.3, blendOption=Linear, blendOutTime=0.8}
    boundBone {boneName=hair_MSTR}
    componentAppliedLinearAccClamp=<750 750 750>
    componentLinearAccScale=<0.9 0.9 0.9>
    componentLinearVelScale=<1 1 1>
    gravityOverride {z=-500}
    physicsBodyDefinitions [{
      sphereCollisionRadius=0
      boundBone {boneName=hair_MSTR}
      boxExtents=<40 40 40>
      constraintSetup {angularTargetAxis=AxisY, angularLimitsMax=<0 0 -15>, angularLimitsMin=<0 0 -15>, angularTarget=<0 0 0>, linearAxesMin=<-4 -1 0>}
      localJointOffset=<0 0 0>
     }]
   }
   n17 = Anim Dynamics - Bone: dyn_hairBack_01_r_MSTR!(ComponentPose=@n16, bAlphaBoolEnabled:bool=@n15, ExternalForce="0,0,0") -> (Pose) {
    alphaInputType=Bool
    angularSpringConstant=80
    bGravityOverrideInSimSpace=true
    bOverrideLinearDamping=true
    bUseGravityOverride=true
    lODThreshold=1
    linearDampingOverride=0.98
    linearSpringConstant=50
    simulationSpace=RootRelative
    alphaBoolBlend {blendInTime=0.1, blendOption=Linear, blendOutTime=0.8}
    boundBone {boneName=dyn_hairBack_01_r_MSTR}
    componentAppliedLinearAccClamp=<750 750 750>
    componentLinearAccScale=<0.9 0.9 0.9>
    componentLinearVelScale=<1 1 1>
    gravityOverride {z=-500}
    physicsBodyDefinitions [{
      sphereCollisionRadius=0
      boundBone {boneName=dyn_hairBack_01_r_MSTR}
      boxExtents=<20 20 20>
      constraintSetup {
       angularTargetAxis=AxisY
       angularLimitsMax=<0 0 -15>
       angularLimitsMin=<0 0 -15>
       angularTarget=<0 0 0>
       linearAxesMax=<1 0 0>
       linearAxesMin=<-2 0 0>
      }
      localJointOffset=<0 0 0>
     }]
   }
   Component To Local(ComponentPose=@n17)
   n18 = Get Head_rotation()
   n19 = "Transform (Modify) Bone - Bone: dyn_hairBack_02_c"(
    ComponentPose=Local To Component(LocalPose=Use cached pose 'AfterAim'())
    Alpha=@n18
    Translation="1,-2,-5"
    Rotation="0.000,20.000,-10.000"
    Scale="1,1,1"
   ) {
    rotationMode=BMM_Additive
    rotationSpace=BCS_BoneSpace
    scaleMode=BMM_Additive
    scaleSpace=BCS_BoneSpace
    translationMode=BMM_Additive
    translationSpace=BCS_BoneSpace
    alphaBoolBlend {blendOption=Linear}
    boneToModify {boneName=dyn_hairBack_02_c}
    rotation=<rot 0 20.0001 -10>
    translation=<1 -2 -5>
   }
   n20 = "Transform (Modify) Bone - Bone: dyn_hairMid_02_r"(ComponentPose=@n19, Alpha=@n18, Translation="-10,0,-0.035", Rotation="0,0,0", Scale="1,1,1") {
    rotationMode=BMM_Additive
    rotationSpace=BCS_BoneSpace
    scaleMode=BMM_Additive
    scaleSpace=BCS_BoneSpace
    translationMode=BMM_Additive
    translationSpace=BCS_BoneSpace
    alphaBoolBlend {blendOption=Linear}
    boneToModify {boneName=dyn_hairMid_02_r}
    translation {x=-10, z=-0.035}
   }
   n21 = "Transform (Modify) Bone - Bone: dyn_hairBack_02_r"(ComponentPose=@n20, Alpha=@n18, Translation="-0.000,0.000,-3.837", Rotation="0,0,0", Scale="1,1,1") {
    rotationMode=BMM_Additive
    rotationSpace=BCS_BoneSpace
    scaleMode=BMM_Additive
    scaleSpace=BCS_BoneSpace
    translationMode=BMM_Additive
    translationSpace=BCS_BoneSpace
    alphaBoolBlend {blendOption=Linear}
    boneToModify {boneName=dyn_hairBack_02_r}
    translation {z=-3.837}
   }
   n22 = "Transform (Modify) Bone - Bone: dyn_hairBack_02_l"(ComponentPose=@n21, Alpha=@n18, Translation="0.000,-7.665,6.851", Rotation="0,0,0", Scale="1,1,1") {
    rotationMode=BMM_Additive
    rotationSpace=BCS_BoneSpace
    scaleMode=BMM_Additive
    scaleSpace=BCS_BoneSpace
    translationMode=BMM_Additive
    translationSpace=BCS_BoneSpace
    alphaBoolBlend {blendOption=Linear}
    boneToModify {boneName=dyn_hairBack_02_l}
    translation {y=-7.665, z=6.851}
   }
   n23 = Component To Local(ComponentPose=@n22)
   n24 = Use cached pose 'AfterAim'()
   n25 = Get bIsPlayingEmote()
   n26 = "Transform (Modify) Bone - Bone: dyn_hairMid_01_l_MSTR"!(
    ComponentPose=Local To Component(LocalPose=Blend Poses by bool(BlendPose_0=@n23, BlendPose_1=@n24, bActiveValue=@n25) {blendType=Linear})
    bAlphaBoolEnabled:bool=Get bIsJumping()
    Translation="0,0,0"
    Rotation="0,10,0"
    Scale="1,1,1"
   ) -> (Pose) {
    alphaInputType=Bool
    rotationMode=BMM_Additive
    rotationSpace=BCS_BoneSpace
    scaleMode=BMM_Additive
    scaleSpace=BCS_BoneSpace
    translationMode=BMM_Additive
    translationSpace=BCS_BoneSpace
    alphaBoolBlend {blendInTime=0.3, blendOption=Linear, blendOutTime=0.8}
    boneToModify {boneName=dyn_hairMid_01_l_MSTR}
    rotation {yaw=10}
   }
   Control Rig!(Source=Component To Local(ComponentPose=@n26), bAlphaBoolEnabled:bool=Get bIsPlayingMeleeAnim()) -> (Pose) {
    alphaInputType=Bool
    alphaBoolBlend {blendInTime=0.1, blendOption=Linear, blendOutTime=0.1}
    controlRigAssetReference {blueprintRigClass=&/F_MED_DenimEquip_Stock_FaceAcc_CtrlRig.F_MED_DenimEquip_Stock_FaceAcc_CtrlRig_C}
   }
   n27 = "Transform (Modify) Bone - Bone: dyn_hairBack_02_c"(Alpha=1, Translation="0.000,-9.792,0.000", Rotation="0,0,0", Scale="1,1,1") {
    rotationMode=BMM_Additive
    rotationSpace=BCS_BoneSpace
    scaleMode=BMM_Additive
    scaleSpace=BCS_BoneSpace
    translationMode=BMM_Additive
    translationSpace=BCS_BoneSpace
    alphaBoolBlend {blendOption=Linear}
    boneToModify {boneName=dyn_hairBack_02_c}
    translation {y=-9.792}
   }
   "Transform (Modify) Bone - Bone: dyn_hairMid_02_r"(ComponentPose=@n27, Alpha=1, Translation="-6.524,4.646,2.554", Rotation="0,0,0", Scale="1,1,1") {
    rotationMode=BMM_Additive
    rotationSpace=BCS_BoneSpace
    scaleMode=BMM_Additive
    scaleSpace=BCS_BoneSpace
    translationMode=BMM_Additive
    translationSpace=BCS_BoneSpace
    alphaBoolBlend {blendOption=Linear}
    boneToModify {boneName=dyn_hairMid_02_r}
    translation=<-6.524 4.646 2.554>
   }
   n28 = Get Head_rotation()
   n29 = "Transform (Modify) Bone - Bone: dyn_hairBack_02_c"(Alpha=@n28, Translation="0,0,0", Rotation="0,0,0", Scale="1,1,1") {
    rotationMode=BMM_Additive
    rotationSpace=BCS_BoneSpace
    scaleMode=BMM_Additive
    scaleSpace=BCS_BoneSpace
    translationMode=BMM_Additive
    translationSpace=BCS_BoneSpace
    alphaBoolBlend {blendOption=Linear}
    boneToModify {boneName=dyn_hairBack_02_c}
   }
   n30 = "Transform (Modify) Bone - Bone: dyn_hairMid_02_r"(ComponentPose=@n29, Alpha=@n28, Translation="-10.116,-0.000,1.058", Rotation="0,0,0", Scale="0.5,0.5,0.5") {
    rotationMode=BMM_Additive
    rotationSpace=BCS_BoneSpace
    scaleMode=BMM_Additive
    scaleSpace=BCS_BoneSpace
    translationMode=BMM_Additive
    translationSpace=BCS_BoneSpace
    alphaBoolBlend {blendOption=Linear}
    boneToModify {boneName=dyn_hairMid_02_r}
    scale=<0.5 0.5 0.5>
    translation {x=-10.116, z=1.058}
   }
   n31 = "Transform (Modify) Bone - Bone: dyn_hairBack_02_r"(ComponentPose=@n30, Alpha=@n28, Translation="-3,-5,-1.062", Rotation="0,50,0", Scale="0.5,0.5,0.5") {
    rotationMode=BMM_Additive
    rotationSpace=BCS_BoneSpace
    scaleMode=BMM_Additive
    scaleSpace=BCS_BoneSpace
    translationMode=BMM_Additive
    translationSpace=BCS_BoneSpace
    alphaBoolBlend {blendOption=Linear}
    boneToModify {boneName=dyn_hairBack_02_r}
    rotation {yaw=50}
    scale=<0.5 0.5 0.5>
    translation=<-3 -5 -1.062>
   }
   "Transform (Modify) Bone - Bone: dyn_hairBack_02_l"(ComponentPose=@n31, Alpha=@n28, Translation="0,0,2", Rotation="0,0,0", Scale="1,1,1") {
    rotationMode=BMM_Additive
    rotationSpace=BCS_BoneSpace
    scaleMode=BMM_Additive
    scaleSpace=BCS_BoneSpace
    translationMode=BMM_Additive
    translationSpace=BCS_BoneSpace
    alphaBoolBlend {blendOption=Linear}
    boneToModify {boneName=dyn_hairBack_02_l}
    translation {z=2}
   }
   "Transform (Modify) Bone - Bone: dyn_hairMid_02_r"(Alpha=Get Head_rotation(), Translation="-60,0,50", Rotation="0,0,0", Scale="1,1,1") {
    rotationMode=BMM_Additive
    rotationSpace=BCS_BoneSpace
    scaleMode=BMM_Additive
    scaleSpace=BCS_BoneSpace
    translationMode=BMM_Additive
    translationSpace=BCS_BoneSpace
    alphaBoolBlend {blendOption=Linear}
    boneToModify {boneName=dyn_hairMid_02_r}
    translation {x=-60, z=50}
   }
   n32 = Get Head_rotation_lowAlpha()
   n33 = "Transform (Modify) Bone - Bone: dyn_hairBack_02_c"(Alpha=@n32, Translation="0,-4.2,-0.238", Rotation="0.028,0.004,30.000", Scale="0.5,0.5,0.5") {
    rotationMode=BMM_Additive
    rotationSpace=BCS_BoneSpace
    scaleMode=BMM_Additive
    scaleSpace=BCS_BoneSpace
    translationMode=BMM_Additive
    translationSpace=BCS_BoneSpace
    alphaBoolBlend {blendOption=Linear}
    boneToModify {boneName=dyn_hairBack_02_c}
    rotation=<rot 0.028 0.004 30>
    scale=<0.5 0.5 0.5>
    translation {y=-4.2, z=-0.238}
   }
   n34 = "Transform (Modify) Bone - Bone: dyn_hairMid_02_r"(ComponentPose=@n33, Alpha=@n32, Translation="-6.996,3.397,9.618", Rotation="0,0,0", Scale="0.7,0.7,0.7") {
    rotationMode=BMM_Additive
    rotationSpace=BCS_BoneSpace
    scaleMode=BMM_Additive
    scaleSpace=BCS_BoneSpace
    translationMode=BMM_Additive
    translationSpace=BCS_BoneSpace
    alphaBoolBlend {blendOption=Linear}
    boneToModify {boneName=dyn_hairMid_02_r}
    scale=<0.7 0.7 0.7>
    translation=<-6.996 3.397 9.618>
   }
   n35 = "Transform (Modify) Bone - Bone: dyn_hairBack_02_r"(ComponentPose=@n34, Alpha=Get Head_rotation(), Translation="0,-5,-5", Rotation="7.727,40.480,6.468", Scale="0.7,0.7,0.7") {
    rotationMode=BMM_Additive
    rotationSpace=BCS_BoneSpace
    scaleMode=BMM_Additive
    scaleSpace=BCS_BoneSpace
    translationMode=BMM_Additive
    translationSpace=BCS_BoneSpace
    alphaBoolBlend {blendOption=Linear}
    boneToModify {boneName=dyn_hairBack_02_r}
    rotation=<rot 7.727 40.48 6.468>
    scale=<0.7 0.7 0.7>
    translation {y=-5, z=-5}
   }
   n36 = "Transform (Modify) Bone - Bone: dyn_hairBack_02_l"(ComponentPose=@n35, Alpha=@n32, Translation="0,0,2", Rotation="0,0,0", Scale="1,1,1") {
    rotationMode=BMM_Additive
    rotationSpace=BCS_BoneSpace
    scaleMode=BMM_Additive
    scaleSpace=BCS_BoneSpace
    translationMode=BMM_Additive
    translationSpace=BCS_BoneSpace
    alphaBoolBlend {blendOption=Linear}
    boneToModify {boneName=dyn_hairBack_02_l}
    translation {z=2}
   }
   "Transform (Modify) Bone - Bone: dyn_hairMid_02_r"!(ComponentPose=@n36, bAlphaBoolEnabled:bool=Get bIsCrouching(), Translation="-4.097,-4.053,1.455", Rotation="0,0,0", Scale="1,1,1") -> (Pose) {
    alphaInputType=Bool
    rotationMode=BMM_Additive
    rotationSpace=BCS_BoneSpace
    scaleMode=BMM_Additive
    scaleSpace=BCS_BoneSpace
    translationMode=BMM_Additive
    translationSpace=BCS_BoneSpace
    alphaBoolBlend {blendInTime=0.1, blendOption=Linear, blendOutTime=0.1}
    boneToModify {boneName=dyn_hairMid_02_r}
    translation=<-4.097 -4.053 1.455>
   }
   n37 = "Transform (Modify) Bone - Bone: dyn_hair_c_02"(Alpha=1, Translation="-2,0,0", Rotation="15,0,0", Scale="1,1,1") {
    rotationMode=BMM_Additive
    rotationSpace=BCS_BoneSpace
    scaleMode=BMM_Additive
    scaleSpace=BCS_BoneSpace
    translationMode=BMM_Additive
    translationSpace=BCS_BoneSpace
    alphaBoolBlend {blendOption=Linear}
    boneToModify {boneName=dyn_hair_c_02}
    rotation {pitch=15}
    translation {x=-2}
   }
   "Transform (Modify) Bone - Bone: dyn_hair_c_01"(ComponentPose=@n37, Alpha=1, Translation="-3,0,0", Rotation="0,0,0", Scale="1,1,1") {
    rotationMode=BMM_Additive
    rotationSpace=BCS_BoneSpace
    scaleMode=BMM_Additive
    scaleSpace=BCS_BoneSpace
    translationMode=BMM_Additive
    translationSpace=BCS_BoneSpace
    alphaBoolBlend {blendOption=Linear}
    boneToModify {boneName=dyn_hair_c_01}
    translation {x=-3}
   }
   n38 = AND Boolean!(A=Get bIsCrouching(), B=Get bIsMoving2D(), C:bool=Get bIsTargeting()) -> (ReturnValue)
   n39 = "Transform (Modify) Bone - Bone: dyn_hair_r_01"!(ComponentPose, bAlphaBoolEnabled:bool=@n38, Translation="-2,0,0", Rotation="25,50,0", Scale="1,1,1") -> (Pose) {
    alphaInputType=Bool
    rotationMode=BMM_Additive
    rotationSpace=BCS_BoneSpace
    scaleMode=BMM_Additive
    scaleSpace=BCS_BoneSpace
    translationMode=BMM_Additive
    translationSpace=BCS_BoneSpace
    alphaBoolBlend {blendInTime=0.1, blendOption=Linear, blendOutTime=0.1}
    boneToModify {boneName=dyn_hair_r_01}
    rotation {pitch=25, yaw=50}
    translation {x=-2}
   }
   n40 = "Transform (Modify) Bone - Bone: dyn_hair_r2_01"!(ComponentPose=@n39, bAlphaBoolEnabled:bool=@n38, Translation="-2,0,0", Rotation="25,70,0", Scale="1,1,1") -> (Pose) {
    alphaInputType=Bool
    rotationMode=BMM_Additive
    rotationSpace=BCS_BoneSpace
    scaleMode=BMM_Additive
    scaleSpace=BCS_BoneSpace
    translationMode=BMM_Additive
    translationSpace=BCS_BoneSpace
    alphaBoolBlend {blendInTime=0.1, blendOption=Linear, blendOutTime=0.1}
    boneToModify {boneName=dyn_hair_r2_01}
    rotation {pitch=25, yaw=70}
    translation {x=-2}
   }
   n41 = "Transform (Modify) Bone - Bone: dyn_hair_c_01"!(ComponentPose=@n40, bAlphaBoolEnabled:bool=@n38, Translation="-2,0,0", Rotation="70,0,0", Scale="1,1,1") -> (Pose) {
    alphaInputType=Bool
    rotationMode=BMM_Additive
    rotationSpace=BCS_BoneSpace
    scaleMode=BMM_Additive
    scaleSpace=BCS_BoneSpace
    translationMode=BMM_Additive
    translationSpace=BCS_BoneSpace
    alphaBoolBlend {blendInTime=0.1, blendOption=Linear, blendOutTime=0.1}
    boneToModify {boneName=dyn_hair_c_01}
    rotation {pitch=70}
    translation {x=-2}
   }
   n42 = "Transform (Modify) Bone - Bone: dyn_hair_l_01"!(ComponentPose=@n41, bAlphaBoolEnabled:bool=@n38, Translation="-2,0,0", Rotation="-10,25,0", Scale="1,1,1") -> (Pose) {
    alphaInputType=Bool
    rotationMode=BMM_Additive
    rotationSpace=BCS_BoneSpace
    scaleMode=BMM_Additive
    scaleSpace=BCS_BoneSpace
    translationMode=BMM_Additive
    translationSpace=BCS_BoneSpace
    alphaBoolBlend {blendInTime=0.1, blendOption=Linear, blendOutTime=0.1}
    boneToModify {boneName=dyn_hair_l_01}
    rotation {pitch=-10, yaw=25}
    translation {x=-2}
   }
   "Transform (Modify) Bone - Bone: dyn_hair_l2_01"!(ComponentPose=@n42, bAlphaBoolEnabled:bool=@n38, Translation="-2,0,0", Rotation="-25,50,0", Scale="1,1,1") -> (Pose) {
    alphaInputType=Bool
    rotationMode=BMM_Additive
    rotationSpace=BCS_BoneSpace
    scaleMode=BMM_Additive
    scaleSpace=BCS_BoneSpace
    translationMode=BMM_Additive
    translationSpace=BCS_BoneSpace
    alphaBoolBlend {blendInTime=0.1, blendOption=Linear, blendOutTime=0.1}
    boneToModify {boneName=dyn_hair_l2_01}
    rotation {pitch=-25, yaw=50}
    translation {x=-2}
   }
   n43 = Get Head_rotation_lowAlpha()
   n44 = "Transform (Modify) Bone - Bone: dyn_hair_r_02"(
    ComponentPose=Local To Component(LocalPose=Use cached pose 'StandLogic'())
    Alpha=@n43
    Translation="-2,0,0"
    Rotation="10,10,0"
    Scale="1,1,1"
   ) {
    rotationMode=BMM_Additive
    rotationSpace=BCS_BoneSpace
    scaleMode=BMM_Additive
    scaleSpace=BCS_BoneSpace
    translationMode=BMM_Additive
    translationSpace=BCS_BoneSpace
    alphaBoolBlend {blendOption=Linear}
    boneToModify {boneName=dyn_hair_r_02}
    rotation {pitch=10, yaw=10}
    translation {x=-2}
   }
   n45 = "Transform (Modify) Bone - Bone: dyn_hair_c_02"(ComponentPose=@n44, Alpha=@n43, Translation="-1,0,0", Rotation="10,0,0", Scale="1,1,1") {
    rotationMode=BMM_Additive
    rotationSpace=BCS_BoneSpace
    scaleMode=BMM_Additive
    scaleSpace=BCS_BoneSpace
    translationMode=BMM_Additive
    translationSpace=BCS_BoneSpace
    alphaBoolBlend {blendOption=Linear}
    boneToModify {boneName=dyn_hair_c_02}
    rotation {pitch=10}
    translation {x=-1}
   }
   n46 = "Transform (Modify) Bone - Bone: dyn_hair_c_02"(ComponentPose=@n45, Alpha=Get Head_rotation(), Translation="-3,0,0", Rotation="0,0,0", Scale="1,1,1") {
    rotationMode=BMM_Additive
    rotationSpace=BCS_BoneSpace
    scaleMode=BMM_Additive
    scaleSpace=BCS_BoneSpace
    translationMode=BMM_Additive
    translationSpace=BCS_BoneSpace
    alphaBoolBlend {blendInTime=0.1, blendOption=Linear, blendOutTime=0.1}
    boneToModify {boneName=dyn_hair_c_02}
    translation {x=-3}
   }
   n47 = Blend Poses by bool(
    BlendPose_0=Component To Local(ComponentPose=@n46)
    BlendPose_1=Use cached pose 'StandLogic'()
    BlendTime_0=0.2
    BlendTime_1=0.3
    bActiveValue=Get CrouchTargeting 2H() -> (CrouchTargeting 2H:bool)
   ) {blendType=Linear}
   n48 = Get Head_rotation_lowAlpha()
   n49 = "Transform (Modify) Bone - Bone: dyn_hair_r_02"(
    ComponentPose=Local To Component(LocalPose=Use cached pose 'StandLogic'())
    Alpha=@n48
    Translation="-2,0,0"
    Rotation="10,10,0"
    Scale="1,1,1"
   ) {
    rotationMode=BMM_Additive
    rotationSpace=BCS_BoneSpace
    scaleMode=BMM_Additive
    scaleSpace=BCS_BoneSpace
    translationMode=BMM_Additive
    translationSpace=BCS_BoneSpace
    alphaBoolBlend {blendOption=Linear}
    boneToModify {boneName=dyn_hair_r_02}
    rotation {pitch=10, yaw=10}
    translation {x=-2}
   }
   n50 = "Transform (Modify) Bone - Bone: dyn_hair_c_02"(ComponentPose=@n49, Alpha=@n48, Translation="-1,0,0", Rotation="10,0,0", Scale="1,1,1") {
    rotationMode=BMM_Additive
    rotationSpace=BCS_BoneSpace
    scaleMode=BMM_Additive
    scaleSpace=BCS_BoneSpace
    translationMode=BMM_Additive
    translationSpace=BCS_BoneSpace
    alphaBoolBlend {blendOption=Linear}
    boneToModify {boneName=dyn_hair_c_02}
    rotation {pitch=10}
    translation {x=-1}
   }
   n51 = Component To Local(ComponentPose=@n50)
   n52 = Get CrouchTargeting()
   n53 = "Transform (Modify) Bone - Bone: dyn_hair_l2_02"!(
    ComponentPose=Local To Component(LocalPose=Blend Poses by bool(BlendPose_0=@n51, BlendPose_1=@n47, bActiveValue=@n52) {blendType=Linear})
    bAlphaBoolEnabled:bool=Get CrouchMoving() -> (CrouchMoving:bool)
    Translation="-2,0,-2"
    Rotation="0,0,0"
    Scale="1,1,1"
   ) -> (Pose) {
    alphaInputType=Bool
    rotationMode=BMM_Additive
    rotationSpace=BCS_BoneSpace
    scaleMode=BMM_Additive
    scaleSpace=BCS_BoneSpace
    translationMode=BMM_Additive
    translationSpace=BCS_BoneSpace
    alphaBoolBlend {blendInTime=0.3, blendOption=Linear, blendOutTime=0.3}
    boneToModify {boneName=dyn_hair_l2_02}
    translation {x=-2, z=-2}
   }
   n54 = "Transform (Modify) Bone - Bone: dyn_hair_c_01"(ComponentPose=@n53, bAlphaBoolEnabled=Get CrouchingIdle() -> (CrouchingIdle:bool), Translation="0,0,0", Rotation="10,0,0", Scale="1,1,1") {
    alphaInputType=Bool
    rotationMode=BMM_Additive
    rotationSpace=BCS_BoneSpace
    scaleMode=BMM_Additive
    scaleSpace=BCS_BoneSpace
    translationMode=BMM_Additive
    translationSpace=BCS_BoneSpace
    alphaBoolBlend {blendInTime=0.3, blendOption=Linear, blendOutTime=0.3}
    boneToModify {boneName=dyn_hair_c_01}
    rotation {pitch=10}
   }
   n55 = Anim Dynamics - Bone: dyn_hair_bang_01(ComponentPose=@n54, bAlphaBoolEnabled=True, ExternalForce="0,0,0") {
    alphaInputType=Bool
    angularDampingOverride=1
    bOverrideAngularDamping=true
    bUseGravityOverride=true
    simulationSpace=RootRelative
    alphaBoolBlend {blendInTime=0.1, blendOption=Linear, blendOutTime=0.2}
    boundBone {boneName=dyn_hair_bang_01}
    componentAppliedLinearAccClamp=<2000 2000 2000>
    componentLinearAccScale=<0.2 0.2 0.2>
    componentLinearVelScale=<0.3 0.3 0.3>
    gravityOverride {z=-700}
    physicsBodyDefinitions [{
      boundBone {boneName=dyn_hair_bang_01}
      boxExtents=<10 10 10>
      constraintSetup {angularLimitsMax=<5 20 0>, angularLimitsMin=<-5 -20 -5>, angularTarget=<0 0 0>}
      localJointOffset=<1 0 0>
     }]
   }
   n56 = Anim Dynamics - Bone: dyn_hair_bang_02(ComponentPose=@n55, bAlphaBoolEnabled=True, ExternalForce="0,0,0") {
    alphaInputType=Bool
    angularDampingOverride=1
    bOverrideAngularDamping=true
    bUseGravityOverride=true
    simulationSpace=RootRelative
    alphaBoolBlend {blendInTime=0.1, blendOption=Linear, blendOutTime=0.2}
    boundBone {boneName=dyn_hair_bang_02}
    componentAppliedLinearAccClamp=<2000 2000 2000>
    componentLinearAccScale=<0.2 0.2 0.2>
    componentLinearVelScale=<0.3 0.3 0.3>
    gravityOverride {z=-700}
    physicsBodyDefinitions [{
      boundBone {boneName=dyn_hair_bang_02}
      boxExtents=<10 10 10>
      constraintSetup {angularLimitsMax=<5 20 0>, angularLimitsMin=<-5 -20 -5>, angularTarget=<0 0 0>}
      localJointOffset=<1 0 0>
     }]
   }
   n57 = Get jump_falling() -> (jump_falling:bool)
   n58 = "Transform (Modify) Bone - Bone: hair_MSTR"!(ComponentPose=@n56, bAlphaBoolEnabled:bool=@n57, Translation="-3,0,0", Rotation="0,0,0", Scale="1,1,1") -> (Pose) {
    alphaInputType=Bool
    rotationMode=BMM_Additive
    rotationSpace=BCS_BoneSpace
    scaleMode=BMM_Additive
    scaleSpace=BCS_BoneSpace
    translationMode=BMM_Additive
    translationSpace=BCS_BoneSpace
    alphaBoolBlend {blendInTime=0.4, blendOption=Linear, blendOutTime=0.1}
    boneToModify {boneName=hair_MSTR}
    translation {x=-3}
   }
   n59 = "Transform (Modify) Bone - Bone: dyn_hair_r_01"(ComponentPose=@n58, bAlphaBoolEnabled=@n57, Translation="0,1,1", Rotation="10,0,0", Scale="1,1,1") {
    alphaInputType=Bool
    rotationMode=BMM_Additive
    rotationSpace=BCS_BoneSpace
    scaleMode=BMM_Additive
    scaleSpace=BCS_BoneSpace
    translationMode=BMM_Additive
    translationSpace=BCS_BoneSpace
    alphaBoolBlend {blendInTime=0.4, blendOption=Linear, blendOutTime=0.1}
    boneToModify {boneName=dyn_hair_r_01}
    rotation {pitch=10}
    translation {y=1, z=1}
   }
   n60 = "Transform (Modify) Bone - Bone: dyn_hair_r2_01"(ComponentPose=@n59, bAlphaBoolEnabled=@n57, Translation="0,1,1", Rotation="10,0,0", Scale="1,1,1") {
    alphaInputType=Bool
    rotationMode=BMM_Additive
    rotationSpace=BCS_BoneSpace
    scaleMode=BMM_Additive
    scaleSpace=BCS_BoneSpace
    translationMode=BMM_Additive
    translationSpace=BCS_BoneSpace
    alphaBoolBlend {blendInTime=0.4, blendOption=Linear, blendOutTime=0.1}
    boneToModify {boneName=dyn_hair_r2_01}
    rotation {pitch=10}
    translation {y=1, z=1}
   }
   n61 = "Transform (Modify) Bone - Bone: dyn_hair_c_01"(ComponentPose=@n60, bAlphaBoolEnabled=@n57, Translation="0,1,1", Rotation="10,0,0", Scale="1,1,1") {
    alphaInputType=Bool
    rotationMode=BMM_Additive
    rotationSpace=BCS_BoneSpace
    scaleMode=BMM_Additive
    scaleSpace=BCS_BoneSpace
    translationMode=BMM_Additive
    translationSpace=BCS_BoneSpace
    alphaBoolBlend {blendInTime=0.4, blendOption=Linear, blendOutTime=0.1}
    boneToModify {boneName=dyn_hair_c_01}
    rotation {pitch=10}
    translation {y=1, z=1}
   }
   n62 = "Transform (Modify) Bone - Bone: dyn_hair_l_01"(ComponentPose=@n61, bAlphaBoolEnabled=@n57, Translation="0,1,-1", Rotation="-10,0,0", Scale="1,1,1") {
    alphaInputType=Bool
    rotationMode=BMM_Additive
    rotationSpace=BCS_BoneSpace
    scaleMode=BMM_Additive
    scaleSpace=BCS_BoneSpace
    translationMode=BMM_Additive
    translationSpace=BCS_BoneSpace
    alphaBoolBlend {blendInTime=0.4, blendOption=Linear, blendOutTime=0.1}
    boneToModify {boneName=dyn_hair_l_01}
    rotation {pitch=-10}
    translation {y=1, z=-1}
   }
   n63 = "Transform (Modify) Bone - Bone: dyn_hair_l2_01"(ComponentPose=@n62, bAlphaBoolEnabled=@n57, Translation="0,1,-1", Rotation="-10,0,0", Scale="1,1,1") {
    alphaInputType=Bool
    rotationMode=BMM_Additive
    rotationSpace=BCS_BoneSpace
    scaleMode=BMM_Additive
    scaleSpace=BCS_BoneSpace
    translationMode=BMM_Additive
    translationSpace=BCS_BoneSpace
    alphaBoolBlend {blendInTime=0.4, blendOption=Linear, blendOutTime=0.1}
    boneToModify {boneName=dyn_hair_l2_01}
    rotation {pitch=-10}
    translation {y=1, z=-1}
   }
   n64 = Get bIsSkydiveFloating()
   n65 = "Transform (Modify) Bone - Bone: dyn_hair_r_01"(ComponentPose=@n63, bAlphaBoolEnabled=@n64, Translation="0,1,1", Rotation="5,0,0", Scale="1,1,1") {
    alphaInputType=Bool
    rotationMode=BMM_Additive
    rotationSpace=BCS_BoneSpace
    scaleMode=BMM_Additive
    scaleSpace=BCS_BoneSpace
    translationMode=BMM_Additive
    translationSpace=BCS_BoneSpace
    alphaBoolBlend {blendInTime=0.1, blendOption=Linear, blendOutTime=0.1}
    boneToModify {boneName=dyn_hair_r_01}
    rotation {pitch=5}
    translation {y=1, z=1}
   }
   n66 = "Transform (Modify) Bone - Bone: dyn_hair_r2_01"(ComponentPose=@n65, bAlphaBoolEnabled=@n64, Translation="0,1,1", Rotation="10,10,0", Scale="1,1,1") {
    alphaInputType=Bool
    rotationMode=BMM_Additive
    rotationSpace=BCS_BoneSpace
    scaleMode=BMM_Additive
    scaleSpace=BCS_BoneSpace
    translationMode=BMM_Additive
    translationSpace=BCS_BoneSpace
    alphaBoolBlend {blendInTime=0.1, blendOption=Linear, blendOutTime=0.1}
    boneToModify {boneName=dyn_hair_r2_01}
    rotation {pitch=10, yaw=10}
    translation {y=1, z=1}
   }
   n67 = "Transform (Modify) Bone - Bone: dyn_hair_c_01"(ComponentPose=@n66, bAlphaBoolEnabled=@n64, Translation="0,1,1", Rotation="10,0,0", Scale="1,1,1") {
    alphaInputType=Bool
    rotationMode=BMM_Additive
    rotationSpace=BCS_BoneSpace
    scaleMode=BMM_Additive
    scaleSpace=BCS_BoneSpace
    translationMode=BMM_Additive
    translationSpace=BCS_BoneSpace
    alphaBoolBlend {blendInTime=0.1, blendOption=Linear, blendOutTime=0.1}
    boneToModify {boneName=dyn_hair_c_01}
    rotation {pitch=10}
    translation {y=1, z=1}
   }
   n68 = "Transform (Modify) Bone - Bone: dyn_hair_l_01"(ComponentPose=@n67, bAlphaBoolEnabled=@n64, Translation="0,1,-1", Rotation="-10,10,0", Scale="1,1,1") {
    alphaInputType=Bool
    rotationMode=BMM_Additive
    rotationSpace=BCS_BoneSpace
    scaleMode=BMM_Additive
    scaleSpace=BCS_BoneSpace
    translationMode=BMM_Additive
    translationSpace=BCS_BoneSpace
    alphaBoolBlend {blendInTime=0.1, blendOption=Linear, blendOutTime=0.1}
    boneToModify {boneName=dyn_hair_l_01}
    rotation {pitch=-10, yaw=10}
    translation {y=1, z=-1}
   }
   n69 = "Transform (Modify) Bone - Bone: dyn_hair_l2_01"(ComponentPose=@n68, bAlphaBoolEnabled=@n64, Translation="-3,1,-1", Rotation="-10,10,0", Scale="1,1,1") {
    alphaInputType=Bool
    rotationMode=BMM_Additive
    rotationSpace=BCS_BoneSpace
    scaleMode=BMM_Additive
    scaleSpace=BCS_BoneSpace
    translationMode=BMM_Additive
    translationSpace=BCS_BoneSpace
    alphaBoolBlend {blendInTime=0.1, blendOption=Linear, blendOutTime=0.1}
    boneToModify {boneName=dyn_hair_l2_01}
    rotation {pitch=-10, yaw=10}
    translation=<-3 1 -1>
   }
   n70 = "Transform (Modify) Bone - Bone: dyn_hair_l2_02"(ComponentPose=@n69, bAlphaBoolEnabled=@n64, Translation="-1,0,0", Rotation="-25,0,0", Scale="1,1,1") {
    alphaInputType=Bool
    rotationMode=BMM_Additive
    rotationSpace=BCS_BoneSpace
    scaleMode=BMM_Additive
    scaleSpace=BCS_BoneSpace
    translationMode=BMM_Additive
    translationSpace=BCS_BoneSpace
    alphaBoolBlend {blendInTime=0.1, blendOption=Linear, blendOutTime=0.1}
    boneToModify {boneName=dyn_hair_l2_02}
    rotation {pitch=-25}
    translation {x=-1}
   }
   n71 = Control Rig!(
    Source=Component To Local(ComponentPose=@n70)
    bAlphaBoolEnabled=True
    HeadOffset:double=Get Head Offset() -> (Head Offset:double)
    Weapon2H:double=Get Weapon 2H() -> (Weapon 2H:double)
   ) -> (Pose) {
    alphaInputType=Bool
    alphaBoolBlend {blendInTime=0.1, blendOption=Linear, blendOutTime=0.1}
    controlRigAssetReference {blueprintRigClass=&F_MED_DenimEquip_Stock_FaceAcc_CtrlRig_C}
   }
   Output Pose(Result=@n71) {layerGroup=None}
   n72 = Anim Dynamics - Bone: dyn_hair_r_01!(
    ComponentPose=Local To Component(LocalPose=Copy Pose From Mesh() {bUseAttachedParent=true})
    bAlphaBoolEnabled:bool=True
    ExternalForce="0,0,0"
   ) -> (Pose) {
    alphaInputType=Bool
    angularDampingOverride=1
    bOverrideAngularDamping=true
    bUseGravityOverride=true
    bUseSphericalLimits=true
    simulationSpace=RootRelative
    alphaBoolBlend {blendInTime=0.1, blendOption=Linear, blendOutTime=0.2}
    boundBone {boneName=dyn_hair_r_01}
    componentAppliedLinearAccClamp=<2000 2000 2000>
    componentLinearAccScale=<0.2 0.2 0.2>
    componentLinearVelScale=<0.3 0.3 0.3>
    gravityOverride {z=-700}
    physicsBodyDefinitions [{
      boundBone {boneName=dyn_hair_r_01}
      boxExtents=<10 10 10>
      constraintSetup {angularLimitsMax=<5 0 0>, angularLimitsMin=<-5 -10 -5>, angularTarget=<0 0 0>}
      localJointOffset=<1 0 0>
     }]
    sphericalLimits [{limitRadius=11, drivingBone {boneName=upperarm_r}, sphereLocalOffset=<-10 0 0>}]
   }
   n73 = Anim Dynamics - Bone: dyn_hair_r_02!(ComponentPose=@n72, bAlphaBoolEnabled:bool=True, ExternalForce="0,0,0") -> (Pose) {
    alphaInputType=Bool
    angularDampingOverride=1
    bOverrideAngularDamping=true
    bUseGravityOverride=true
    bUsePlanarLimit=false
    simulationSpace=RootRelative
    alphaBoolBlend {blendInTime=0.2, blendOption=Linear, blendOutTime=0.2}
    boundBone {boneName=dyn_hair_r_02}
    componentAppliedLinearAccClamp=<2000 2000 2000>
    componentLinearAccScale=<0.2 0.2 0.2>
    componentLinearVelScale=<0.3 0.3 0.3>
    gravityOverride {z=-700}
    physicsBodyDefinitions [{
      boundBone {boneName=dyn_hair_r_02}
      boxExtents=<10 10 10>
      constraintSetup {angularLimitsMax=<5 0 -5>, angularLimitsMin=<-5 -20 -5>, angularTarget=<0 0 0>}
      localJointOffset=<1 0 0>
     }]
   }
   n74 = Anim Dynamics - Bone: dyn_hair_r2_01!(ComponentPose=@n73, bAlphaBoolEnabled:bool=True, ExternalForce="0,0,0") -> (Pose) {
    alphaInputType=Bool
    angularDampingOverride=1
    bOverrideAngularDamping=true
    bUseGravityOverride=true
    simulationSpace=RootRelative
    alphaBoolBlend {blendInTime=0.1, blendOption=Linear, blendOutTime=0.2}
    boundBone {boneName=dyn_hair_r2_01}
    componentAppliedLinearAccClamp=<2000 2000 2000>
    componentLinearAccScale=<0.2 0.2 0.2>
    componentLinearVelScale=<0.3 0.3 0.3>
    gravityOverride {z=-700}
    physicsBodyDefinitions [{
      boundBone {boneName=dyn_hair_r2_01}
      boxExtents=<10 10 10>
      constraintSetup {angularLimitsMax=<5 0 0>, angularLimitsMin=<-5 -10 -5>, angularTarget=<0 0 0>}
      localJointOffset=<1 0 0>
     }]
    sphericalLimits [{limitRadius=11, drivingBone {boneName=upperarm_r}, sphereLocalOffset=<-10 0 0>}]
   }
   n75 = Anim Dynamics - Bone: dyn_hair_r2_02!(ComponentPose=@n74, bAlphaBoolEnabled:bool=True, ExternalForce="0,0,0") -> (Pose) {
    alphaInputType=Bool
    angularDampingOverride=1
    bOverrideAngularDamping=true
    bUseGravityOverride=true
    bUseSphericalLimits=true
    simulationSpace=RootRelative
    alphaBoolBlend {blendInTime=0.1, blendOption=Linear, blendOutTime=0.2}
    boundBone {boneName=dyn_hair_r2_02}
    componentAppliedLinearAccClamp=<2000 2000 2000>
    componentLinearAccScale=<0.2 0.2 0.2>
    componentLinearVelScale=<0.3 0.3 0.3>
    gravityOverride {z=-700}
    physicsBodyDefinitions [{
      boundBone {boneName=dyn_hair_r2_02}
      boxExtents=<10 10 10>
      constraintSetup {angularLimitsMax=<5 0 0>, angularLimitsMin=<-5 -20 -5>, angularTarget=<0 0 0>}
      localJointOffset=<1 0 0>
     }]
    sphericalLimits [{limitRadius=10, drivingBone {boneName=spine_05}, sphereLocalOffset=<0 0 0>}]
   }
   n76 = Anim Dynamics - Bone: dyn_hair_c_01!(ComponentPose=@n75, bAlphaBoolEnabled:bool=True, ExternalForce="0,0,0") -> (Pose) {
    alphaInputType=Bool
    angularDampingOverride=1
    bOverrideAngularDamping=true
    bUseGravityOverride=true
    bUsePlanarLimit=false
    simulationSpace=RootRelative
    alphaBoolBlend {blendInTime=0.1, blendOption=Linear, blendOutTime=0.2}
    boundBone {boneName=dyn_hair_c_01}
    componentAppliedLinearAccClamp=<2000 2000 2000>
    componentLinearAccScale=<0.2 0.2 0.2>
    componentLinearVelScale=<0.3 0.3 0.3>
    gravityOverride {z=-700}
    physicsBodyDefinitions [{
      boundBone {boneName=dyn_hair_c_01}
      boxExtents=<10 10 10>
      constraintSetup {angularLimitsMax=<5 0 0>, angularLimitsMin=<-5 -10 -5>, angularTarget=<0 0 0>}
      localJointOffset=<1 0 0>
     }]
   }
   n77 = Anim Dynamics - Bone: dyn_hair_l2_01!(ComponentPose=@n76, bAlphaBoolEnabled:bool=True, ExternalForce="0,0,0") -> (Pose) {
    alphaInputType=Bool
    angularDampingOverride=1
    bOverrideAngularDamping=true
    bUseGravityOverride=true
    simulationSpace=RootRelative
    alphaBoolBlend {blendInTime=0.1, blendOption=Linear, blendOutTime=0.2}
    boundBone {boneName=dyn_hair_l2_01}
    componentAppliedLinearAccClamp=<2000 2000 2000>
    componentLinearAccScale=<0.2 0.2 0.2>
    componentLinearVelScale=<0.3 0.3 0.3>
    gravityOverride {z=-700}
    physicsBodyDefinitions [{
      boundBone {boneName=dyn_hair_l2_01}
      boxExtents=<10 10 10>
      constraintSetup {angularLimitsMax=<5 10 0>, angularLimitsMin=<-5 0 -5>, angularTarget=<0 0 0>}
      localJointOffset=<1 0 0>
     }]
    sphericalLimits [{limitRadius=11, drivingBone {boneName=upperarm_r}, sphereLocalOffset=<-10 0 0>}]
   }
   n78 = Anim Dynamics - Bone: dyn_hair_l2_02!(ComponentPose=@n77, bAlphaBoolEnabled:bool=True, ExternalForce="0,0,0") -> (Pose) {
    alphaInputType=Bool
    angularDampingOverride=1
    bOverrideAngularDamping=true
    bUseGravityOverride=true
    simulationSpace=RootRelative
    alphaBoolBlend {blendInTime=0.1, blendOption=Linear, blendOutTime=0.2}
    boundBone {boneName=dyn_hair_l2_02}
    componentAppliedLinearAccClamp=<2000 2000 2000>
    componentLinearAccScale=<0.2 0.2 0.2>
    componentLinearVelScale=<0.3 0.3 0.3>
    gravityOverride {z=-700}
    physicsBodyDefinitions [{
      boundBone {boneName=dyn_hair_l2_02}
      boxExtents=<10 10 10>
      constraintSetup {angularLimitsMax=<5 20 0>, angularLimitsMin=<-5 0 -5>, angularTarget=<0 0 0>}
      localJointOffset=<1 0 0>
     }]
    sphericalLimits [{limitRadius=11, drivingBone {boneName=upperarm_r}, sphereLocalOffset=<-10 0 0>}]
   }
   n79 = Anim Dynamics - Bone: dyn_hair_l_01!(ComponentPose=@n78, bAlphaBoolEnabled:bool=True, ExternalForce="0,0,0") -> (Pose) {
    alphaInputType=Bool
    angularDampingOverride=1
    bOverrideAngularDamping=true
    bUseGravityOverride=true
    simulationSpace=RootRelative
    alphaBoolBlend {blendInTime=0.1, blendOption=Linear, blendOutTime=0.2}
    boundBone {boneName=dyn_hair_l_01}
    componentAppliedLinearAccClamp=<2000 2000 2000>
    componentLinearAccScale=<0.2 0.2 0.2>
    componentLinearVelScale=<0.3 0.3 0.3>
    gravityOverride {z=-700}
    physicsBodyDefinitions [{
      boundBone {boneName=dyn_hair_l_01}
      boxExtents=<10 10 10>
      constraintSetup {angularLimitsMax=<5 10 0>, angularLimitsMin=<-5 0 -5>, angularTarget=<0 0 0>}
      localJointOffset=<1 0 0>
     }]
    sphericalLimits [{limitRadius=11, drivingBone {boneName=upperarm_r}, sphereLocalOffset=<-10 0 0>}]
   }
   n80 = Anim Dynamics - Bone: dyn_hair_l_02!(ComponentPose=@n79, bAlphaBoolEnabled:bool=True, ExternalForce="0,0,0") -> (Pose) {
    alphaInputType=Bool
    angularDampingOverride=1
    bOverrideAngularDamping=true
    bUseGravityOverride=true
    simulationSpace=RootRelative
    alphaBoolBlend {blendInTime=0.1, blendOption=Linear, blendOutTime=0.2}
    boundBone {boneName=dyn_hair_l_02}
    componentAppliedLinearAccClamp=<2000 2000 2000>
    componentLinearAccScale=<0.2 0.2 0.2>
    componentLinearVelScale=<0.3 0.3 0.3>
    gravityOverride {z=-700}
    physicsBodyDefinitions [{
      boundBone {boneName=dyn_hair_l_02}
      boxExtents=<10 10 10>
      constraintSetup {angularLimitsMax=<5 20 0>, angularLimitsMin=<-5 0 -5>, angularTarget=<0 0 0>}
      localJointOffset=<1 0 0>
     }]
    sphericalLimits [{limitRadius=11, drivingBone {boneName=upperarm_r}, sphereLocalOffset=<-10 0 0>}]
   }
   Save cached pose 'Presim'(Pose=Component To Local(ComponentPose=@n80)) {cacheName=Presim}
   n81 = Get Head_rotation()
   n82 = "Transform (Modify) Bone - Bone: dyn_hair_c_01"!(
    ComponentPose=Local To Component(LocalPose=Use cached pose 'Presim'())
    Alpha:float=@n81
    Translation="0,1,1"
    Rotation="20,0,0"
    Scale="1,1,1"
   ) -> (Pose) {
    rotationMode=BMM_Additive
    rotationSpace=BCS_BoneSpace
    scaleMode=BMM_Additive
    scaleSpace=BCS_BoneSpace
    translationMode=BMM_Additive
    translationSpace=BCS_BoneSpace
    alphaBoolBlend {blendOption=Linear}
    boneToModify {boneName=dyn_hair_c_01}
    rotation {pitch=20}
    translation {y=1, z=1}
   }
   n83 = "Transform (Modify) Bone - Bone: dyn_hair_c_02"(ComponentPose=@n82, Alpha=@n81, Translation="-2,0,0", Rotation="10,0,0", Scale="1,1,1") {
    rotationMode=BMM_Additive
    rotationSpace=BCS_BoneSpace
    scaleMode=BMM_Additive
    scaleSpace=BCS_BoneSpace
    translationMode=BMM_Additive
    translationSpace=BCS_BoneSpace
    alphaBoolBlend {blendOption=Linear}
    boneToModify {boneName=dyn_hair_c_02}
    rotation {pitch=10}
    translation {x=-2}
   }
   n84 = "Transform (Modify) Bone - Bone: dyn_hair_r2_01"!(ComponentPose=@n83, Alpha:float=@n81, Translation="0,0,0", Rotation="0,10,0", Scale="1,1,1") -> (Pose) {
    rotationMode=BMM_Additive
    rotationSpace=BCS_BoneSpace
    scaleMode=BMM_Additive
    scaleSpace=BCS_BoneSpace
    translationMode=BMM_Additive
    translationSpace=BCS_BoneSpace
    alphaBoolBlend {blendOption=Linear}
    boneToModify {boneName=dyn_hair_r2_01}
    rotation {yaw=10}
   }
   n85 = "Transform (Modify) Bone - Bone: dyn_hair_r2_02"(ComponentPose=@n84, Alpha=@n81, Translation="0,0,0", Rotation="-0.000,10.000,0.000", Scale="1,1,1") {
    rotationMode=BMM_Additive
    rotationSpace=BCS_BoneSpace
    scaleMode=BMM_Additive
    scaleSpace=BCS_BoneSpace
    translationMode=BMM_Additive
    translationSpace=BCS_BoneSpace
    alphaBoolBlend {blendOption=Linear}
    boneToModify {boneName=dyn_hair_r2_02}
    rotation=<rot 0 10 0>
   }
   n86 = "Transform (Modify) Bone - Bone: dyn_hair_l2_01"!(ComponentPose=@n85, Alpha:float=@n81, Translation="-1.5,0,0", Rotation="0,-5,0", Scale="1,1,1") -> (Pose) {
    rotationMode=BMM_Additive
    rotationSpace=BCS_BoneSpace
    scaleMode=BMM_Additive
    scaleSpace=BCS_BoneSpace
    translationMode=BMM_Additive
    translationSpace=BCS_BoneSpace
    alphaBoolBlend {blendOption=Linear}
    boneToModify {boneName=dyn_hair_l2_01}
    rotation {yaw=-5}
    translation {x=-1.5}
   }
   n87 = "Transform (Modify) Bone - Bone: dyn_hair_l2_02"!(ComponentPose=@n86, Alpha:float=@n81, Translation="-0.654,1.715,-1.798", Rotation="-20.000,-0.000,0.000", Scale="1,1,1") -> (Pose) {
    rotationMode=BMM_Additive
    rotationSpace=BCS_BoneSpace
    scaleMode=BMM_Additive
    scaleSpace=BCS_BoneSpace
    translationMode=BMM_Additive
    translationSpace=BCS_BoneSpace
    alphaBoolBlend {blendOption=Linear}
    boneToModify {boneName=dyn_hair_l2_02}
    rotation=<rot -20 -0.0001 0>
    translation=<-0.654 1.715 -1.798>
   }
   n88 = Get Head_rotation_lowAlpha()
   n89 = "Transform (Modify) Bone - Bone: dyn_hair_r2_02"(
    ComponentPose=Local To Component(LocalPose=Use cached pose 'Presim'())
    Alpha=@n88
    Translation="0,0,0"
    Rotation="0,0,0"
    Scale="0.5,0.5,0.5"
   ) {
    rotationMode=BMM_Additive
    rotationSpace=BCS_BoneSpace
    scaleMode=BMM_Additive
    scaleSpace=BCS_BoneSpace
    translationMode=BMM_Additive
    translationSpace=BCS_BoneSpace
    alphaBoolBlend {blendOption=Linear}
    boneToModify {boneName=dyn_hair_r2_02}
    scale=<0.5 0.5 0.5>
   }
   n90 = Get Head_rotation_3() -> (Head_rotation_3:float)
   n91 = "Transform (Modify) Bone - Bone: dyn_hair_c_02"(ComponentPose=@n89, Alpha=@n90, Translation="-2,0,0", Rotation="15,-10,0", Scale="1,1,1") {
    rotationMode=BMM_Additive
    rotationSpace=BCS_BoneSpace
    scaleMode=BMM_Additive
    scaleSpace=BCS_BoneSpace
    translationMode=BMM_Additive
    translationSpace=BCS_BoneSpace
    alphaBoolBlend {blendOption=Linear}
    boneToModify {boneName=dyn_hair_c_02}
    rotation {pitch=15, yaw=-10}
    translation {x=-2}
   }
   n92 = "Transform (Modify) Bone - Bone: dyn_hair_c_01"(ComponentPose=@n91, Alpha=@n90, Translation="-3,0,0", Rotation="0,0,0", Scale="1,1,1") {
    rotationMode=BMM_Additive
    rotationSpace=BCS_BoneSpace
    scaleMode=BMM_Additive
    scaleSpace=BCS_BoneSpace
    translationMode=BMM_Additive
    translationSpace=BCS_BoneSpace
    alphaBoolBlend {blendOption=Linear}
    boneToModify {boneName=dyn_hair_c_01}
    translation {x=-3}
   }
   n93 = "Transform (Modify) Bone - Bone: dyn_hair_r2_01"(ComponentPose=@n92, Alpha=@n88, Translation="-2.5,0,0", Rotation="0,0,0", Scale="1,1,1") {
    rotationMode=BMM_Additive
    rotationSpace=BCS_BoneSpace
    scaleMode=BMM_Additive
    scaleSpace=BCS_BoneSpace
    translationMode=BMM_Additive
    translationSpace=BCS_BoneSpace
    alphaBoolBlend {blendOption=Linear}
    boneToModify {boneName=dyn_hair_r2_01}
    translation {x=-2.5}
   }
   n94 = "Transform (Modify) Bone - Bone: dyn_hair_r_01"!(ComponentPose=@n93, Alpha:float=@n88, Translation="-0.5,0,0", Rotation="0,0,0", Scale="1,1,1") -> (Pose) {
    rotationMode=BMM_Additive
    rotationSpace=BCS_BoneSpace
    scaleMode=BMM_Additive
    scaleSpace=BCS_BoneSpace
    translationMode=BMM_Additive
    translationSpace=BCS_BoneSpace
    alphaBoolBlend {blendOption=Linear}
    boneToModify {boneName=dyn_hair_r_01}
    translation {x=-0.5}
   }
   n95 = "Transform (Modify) Bone - Bone: dyn_hair_l2_01"(ComponentPose=@n94, Alpha=@n88, Translation="-1,0,0", Rotation="0,0,0", Scale="1,1,1") {
    rotationMode=BMM_Additive
    rotationSpace=BCS_BoneSpace
    scaleMode=BMM_Additive
    scaleSpace=BCS_BoneSpace
    translationMode=BMM_Additive
    translationSpace=BCS_BoneSpace
    alphaBoolBlend {blendOption=Linear}
    boneToModify {boneName=dyn_hair_l2_01}
    translation {x=-1}
   }
   n96 = "Transform (Modify) Bone - Bone: dyn_hair_l_01"!(ComponentPose=@n95, Alpha:float=@n88, Translation="-1,0,0", Rotation="0,0,0", Scale="1,1,1") -> (Pose) {
    rotationMode=BMM_Additive
    rotationSpace=BCS_BoneSpace
    scaleMode=BMM_Additive
    scaleSpace=BCS_BoneSpace
    translationMode=BMM_Additive
    translationSpace=BCS_BoneSpace
    alphaBoolBlend {blendOption=Linear}
    boneToModify {boneName=dyn_hair_l_01}
    translation {x=-1}
   }
   Save cached pose 'StandLogic'(Pose=Blend Poses by bool(BlendPose_0=Component To Local(ComponentPose=@n87), BlendPose_1=Component To Local(ComponentPose=@n96), bActiveValue=Get bIsPlayingEmote()) {blendType=Linear}) {cacheName=StandLogic}
   n97 = Get Head_rotation_lowAlpha()
   n98 = "Transform (Modify) Bone - Bone: dyn_hairBack_02_c"(
    ComponentPose=Local To Component(LocalPose=Use cached pose 'aftersim'())
    Alpha=@n97
    Translation="0,-4.2,-0.238"
    Rotation="0.028,0.004,30.000"
    Scale="0.5,0.5,0.5"
   ) {
    rotationMode=BMM_Additive
    rotationSpace=BCS_BoneSpace
    scaleMode=BMM_Additive
    scaleSpace=BCS_BoneSpace
    translationMode=BMM_Additive
    translationSpace=BCS_BoneSpace
    alphaBoolBlend {blendOption=Linear}
    boneToModify {boneName=dyn_hairBack_02_c}
    rotation=<rot 0.028 0.004 30>
    scale=<0.5 0.5 0.5>
    translation {y=-4.2, z=-0.238}
   }
   n99 = "Transform (Modify) Bone - Bone: dyn_hairMid_02_r"!(ComponentPose=@n98, Alpha:float=@n97, Translation="0,3.397,13.847", Rotation="0,0,0", Scale="0.7,0.7,0.7") -> (Pose) {
    rotationMode=BMM_Additive
    rotationSpace=BCS_BoneSpace
    scaleMode=BMM_Additive
    scaleSpace=BCS_BoneSpace
    translationMode=BMM_Additive
    translationSpace=BCS_BoneSpace
    alphaBoolBlend {blendOption=Linear}
    boneToModify {boneName=dyn_hairMid_02_r}
    scale=<0.7 0.7 0.7>
    translation {y=3.397, z=13.847}
   }
   n100 = "Transform (Modify) Bone - Bone: dyn_hairBack_02_r"(ComponentPose=@n99, Alpha=Get Head_rotation(), Translation="0,-5,-5", Rotation="7.727,40.480,6.468", Scale="0.7,0.7,0.7") {
    rotationMode=BMM_Additive
    rotationSpace=BCS_BoneSpace
    scaleMode=BMM_Additive
    scaleSpace=BCS_BoneSpace
    translationMode=BMM_Additive
    translationSpace=BCS_BoneSpace
    alphaBoolBlend {blendOption=Linear}
    boneToModify {boneName=dyn_hairBack_02_r}
    rotation=<rot 7.727 40.48 6.468>
    scale=<0.7 0.7 0.7>
    translation {y=-5, z=-5}
   }
   n101 = "Transform (Modify) Bone - Bone: dyn_hairBack_02_l"(ComponentPose=@n100, Alpha=@n97, Translation="0,0,2", Rotation="0,0,0", Scale="1,1,1") {
    rotationMode=BMM_Additive
    rotationSpace=BCS_BoneSpace
    scaleMode=BMM_Additive
    scaleSpace=BCS_BoneSpace
    translationMode=BMM_Additive
    translationSpace=BCS_BoneSpace
    alphaBoolBlend {blendOption=Linear}
    boneToModify {boneName=dyn_hairBack_02_l}
    translation {z=2}
   }
   n102 = "Transform (Modify) Bone - Bone: dyn_hairMid_02_r"!(ComponentPose=@n101, bAlphaBoolEnabled:bool=Get bIsCrouching(), Translation="-4.097,-4.053,1.455", Rotation="0,0,0", Scale="1,1,1") -> (Pose) {
    alphaInputType=Bool
    rotationMode=BMM_Additive
    rotationSpace=BCS_BoneSpace
    scaleMode=BMM_Additive
    scaleSpace=BCS_BoneSpace
    translationMode=BMM_Additive
    translationSpace=BCS_BoneSpace
    alphaBoolBlend {blendInTime=0.1, blendOption=Linear, blendOutTime=0.1}
    boneToModify {boneName=dyn_hairMid_02_r}
    translation=<-4.097 -4.053 1.455>
   }
   Save cached pose 'AfterAim'(Pose=Blend Poses by bool(
     BlendPose_0=Component To Local(ComponentPose=@n102)
     BlendPose_1=Use cached pose 'aftersim'()
     BlendTime_1=0.5
     bActiveValue=Get bIsTargeting()
    ) {blendType=Linear}) {cacheName=AfterAim}
   n103 = Anim Dynamics - Bone: dyn_hairBack_01_c(ComponentPose=Local To Component(LocalPose=Use cached pose 'Presim'()), bAlphaBoolEnabled=True, ExternalForce="0,0,0") {
    alphaInputType=Bool
    angularDampingOverride=1
    angularSpringConstant=20000
    bOverrideAngularDamping=true
    bUseGravityOverride=true
    simulationSpace=RootRelative
    alphaBoolBlend {blendInTime=0.1, blendOption=Linear, blendOutTime=0.2}
    boundBone {boneName=dyn_hairBack_01_c}
    componentAppliedLinearAccClamp=<100 100 100>
    componentLinearAccScale=<0.2 0.2 0.2>
    componentLinearVelScale=<0.3 0.3 0.3>
    gravityOverride {z=-900}
    physicsBodyDefinitions [{
      boundBone {boneName=dyn_hairBack_01_c}
      boxExtents=<8 8 8>
      constraintSetup {angularLimitsMax=<4 20 0>, angularLimitsMin=<-4 -20 -25>, angularTarget=<0 0 0>}
      localJointOffset=<1 0 0>
     }]
   }
   n104 = Anim Dynamics - Bone: dyn_hairMid_01_r!(ComponentPose=@n103, bAlphaBoolEnabled:bool=True, ExternalForce="0,0,0") -> (Pose) {
    alphaInputType=Bool
    angularDampingOverride=1
    angularSpringConstant=20000
    bOverrideAngularDamping=true
    bUseGravityOverride=true
    simulationSpace=RootRelative
    alphaBoolBlend {blendInTime=0.1, blendOption=Linear, blendOutTime=0.2}
    boundBone {boneName=dyn_hairMid_01_r}
    componentAppliedLinearAccClamp=<100 100 100>
    componentLinearAccScale=<0.2 0.2 0.2>
    componentLinearVelScale=<0.3 0.3 0.3>
    gravityOverride {z=-900}
    physicsBodyDefinitions [{
      boundBone {boneName=dyn_hairMid_01_r}
      boxExtents=<8 8 8>
      constraintSetup {angularLimitsMax=<4 0 15>, angularLimitsMin=<-4 -10 -15>, angularTarget=<0 0 0>}
      localJointOffset=<1 0 0>
     }]
   }
   n105 = Anim Dynamics - Bone: dyn_hairMid_01_l!(ComponentPose=@n104, bAlphaBoolEnabled:bool=True, ExternalForce="0,0,0") -> (Pose) {
    alphaInputType=Bool
    angularDampingOverride=1
    angularSpringConstant=20000
    bOverrideAngularDamping=true
    bUseGravityOverride=true
    simulationSpace=RootRelative
    alphaBoolBlend {blendInTime=0.1, blendOption=Linear, blendOutTime=0.2}
    boundBone {boneName=dyn_hairMid_01_l}
    componentAppliedLinearAccClamp=<100 100 100>
    componentLinearAccScale=<0.2 0.2 0.2>
    componentLinearVelScale=<0.3 0.3 0.3>
    gravityOverride {z=-900}
    physicsBodyDefinitions [{
      boundBone {boneName=dyn_hairMid_01_l}
      boxExtents=<8 8 8>
      constraintSetup {angularLimitsMax=<4 5 15>, angularLimitsMin=<-4 -5 -15>, angularTarget=<0 0 0>}
      localJointOffset=<1 0 0>
     }]
   }
   n106 = Anim Dynamics - Bone: dyn_hairBack_01_l!(ComponentPose=@n105, bAlphaBoolEnabled:bool=True, ExternalForce="0,0,0") -> (Pose) {
    alphaInputType=Bool
    angularDampingOverride=1
    angularSpringConstant=20000
    bOverrideAngularDamping=true
    bUseGravityOverride=true
    simulationSpace=RootRelative
    alphaBoolBlend {blendInTime=0.1, blendOption=Linear, blendOutTime=0.2}
    boundBone {boneName=dyn_hairBack_01_l}
    componentAppliedLinearAccClamp=<100 100 100>
    componentLinearAccScale=<0.2 0.2 0.2>
    componentLinearVelScale=<0.3 0.3 0.3>
    gravityOverride {z=-900}
    physicsBodyDefinitions [{
      boundBone {boneName=dyn_hairBack_01_l}
      boxExtents=<8 8 8>
      constraintSetup {angularLimitsMax=<4 0 10>, angularLimitsMin=<-4 -10 -10>, angularTarget=<0 0 0>}
      localJointOffset=<1 0 0>
     }]
   }
   n107 = Anim Dynamics - Bone: dyn_hairBack_01_r!(ComponentPose=@n106, bAlphaBoolEnabled:bool=True, ExternalForce="0,0,0") -> (Pose) {
    alphaInputType=Bool
    angularDampingOverride=1
    angularSpringConstant=20000
    bOverrideAngularDamping=true
    bUseGravityOverride=true
    simulationSpace=RootRelative
    alphaBoolBlend {blendInTime=0.1, blendOption=Linear, blendOutTime=0.2}
    boundBone {boneName=dyn_hairBack_01_r}
    componentAppliedLinearAccClamp=<100 100 100>
    componentLinearAccScale=<0.2 0.2 0.2>
    componentLinearVelScale=<0.3 0.3 0.3>
    gravityOverride {z=-900}
    physicsBodyDefinitions [{
      boundBone {boneName=dyn_hairBack_01_r}
      boxExtents=<8 8 8>
      constraintSetup {angularLimitsMax=<4 10 10>, angularLimitsMin=<-4 0 -10>, angularTarget=<0 0 0>}
      localJointOffset=<1 0 0>
     }]
   }
   n108 = RigidBody!(ComponentPose=Local To Component(LocalPose=Use cached pose 'Presim'()), Alpha=0.9, OverrideWorldGravity:Vector="0,0,0", ExternalForce) -> (Pose) {
    bOverrideWorldGravity=true
    overridePhysicsAsset=&/F_MED_DenimEquip_Stock_FaceAcc_Physics
    simulationSpace=BaseBoneSpace
    alphaBoolBlend {blendOption=Linear}
    baseBoneRef {boneName=head}
    componentAppliedLinearAccClamp=<2000 2000 2000>
    componentLinearAccScale=<0.1 0.1 0.1>
    componentLinearVelScale=<1 1 1>
    simSpaceSettings {maxAngularAcceleration=50, maxAngularVelocity=5, maxLinearAcceleration=50, maxLinearVelocity=5, velocityScaleZ=0, worldAlpha=1}
   }
   Save cached pose 'aftersim'(Pose=Blend Poses by bool(
     BlendPose_0=Component To Local(ComponentPose=@n108)
     BlendPose_1=Component To Local(ComponentPose=@n107)
     bActiveValue=Get bRigidBodyEnabled()
    ) {blendType=Linear}) {cacheName=aftersim}
   Use cached pose 'aftersim'()
   Use cached pose 'aftersim'()
   Get NotTargeting() -> (NotTargeting:bool)
  ]}
 EventGraph {nodes [
   n1 = Event BlueprintUpdateAnimation()
   > Set Delta Time X(Delta Time X=@n1.DeltaTimeX)
   n2 = float * float(
    A=GetAlphaFromReferenceTransformDelta(SkeletalMeshComponent=GetOwnerSkeletalMeshComponent(), BoneName=head, BaseName=neck_02, InputRange=(X=0,Y=0), TwistAxis="1,0,0").RawOutput
    B=-0.04
   ) {is=title K2Node_PromotableOperator}
   > Set Head_rotation(Head_rotation:double=FInterpTo(Current=Get Head_rotation(), Target="Clamp (Float)"(Value=@n2, Min=0, Max=1), DeltaTime=@n1.DeltaTimeX, InterpSpeed=10)) -> (Output_Get:double)
   n3 = Get bIsTargeting()
   > Set Head Offset(Head Offset:double=FInterpTo(Current=Get Head Offset(), Target="To Float (Boolean)"(InBool=@n3), DeltaTime=Get Delta Time X(), InterpSpeed=10)) -> (Output_Get:double)
   n4 = AND Boolean!(A=NOT Boolean(A=Get bIsCrouching()), B=Get 2H_range_weapon() -> (2H_range_weapon:bool)) -> (ReturnValue)
   > Set Weapon 2H(Weapon 2H:double=FInterpTo(Current=Get Weapon 2H(), Target="To Float (Boolean)"(InBool=@n4), DeltaTime=Get Delta Time X(), InterpSpeed=10)) -> (Output_Get:double)
   n5 = float * float(
    A=GetAlphaFromReferenceTransformDelta(SkeletalMeshComponent=GetOwnerSkeletalMeshComponent(), BoneName=head, BaseName=neck_02, InputRange=(X=0,Y=0), TwistAxis="1,0,0").RawOutput
    B=-0.02
   )
   > Set Head_rotation_lowAlpha(Head_rotation_lowAlpha:double=FInterpTo(Current=Get Head_rotation(), Target="Clamp (Float)"(Value=@n5, Min=0, Max=1), DeltaTime=0, InterpSpeed=10)) -> (Output_Get:double)
   > n6 = Gravity Override State Function DEPRECATED(CharacterPartAnimInstance=Self-Reference(), GameTimeInSeconds=GetGameTimeInSeconds(), GravityOverrideParams=Get Gravity Override Params())
   > Set Gravity Override(Gravity Override=@n6)
   n7 = Get bIsSkyDiving()
   > Set jump_falling(jump_falling:bool=AND Boolean(A=Get bIsFalling(), B=NOT Boolean(A=@n7))) -> (Output_Get:bool)
   n8 = Get 2H_range_weapon()
   > Set CrouchTargeting(CrouchTargeting:bool=AND Boolean!(A=Get bIsCrouching(), B=Get bIsTargeting(), C:bool=NOT Boolean(A=@n8)) -> (ReturnValue)) -> (Output_Get:bool)
   > Set CrouchMoving(CrouchMoving:bool=AND Boolean!(A=Get bIsCrouching(), B=Get bIsMoving2D()) -> (ReturnValue)) -> (Output_Get:bool)
   n9 = Get bIsTargeting()
   n10 = Get bIsMoving2D()
   > Set CrouchingIdle(CrouchingIdle:bool=AND Boolean!(A=Get bIsCrouching(), B=NOT Boolean(A=@n10), C:bool=NOT Boolean(A=@n9)) -> (ReturnValue)) -> (Output_Get:bool)
   > Set Head_rotation_3(Head_rotation_3:float=GetAlphaFromReferenceTransformDelta(SkeletalMeshComponent=GetOwnerSkeletalMeshComponent(), BoneName=head, BaseName=neck_02, InputRange=(X=-15,Y=-30), TwistAxis="1,0,0")) -> (Output_Get:float)
   > Set NotTargeting(NotTargeting:bool=NOT Boolean(A=Get bIsTargeting())) -> (Output_Get:bool)
   n11 = Get bIsMoving2D()
   > Set CrouchTargeting 2H(CrouchTargeting 2H:bool=AND Boolean!(A=Get bIsCrouching(), B=Get bIsTargeting(), C=Get 2H_range_weapon(), D:bool=NOT Boolean(A=@n11)) -> (ReturnValue)) -> (Output_Get:bool)
   > n12 = Cast To FortPlayerAnimInstance(Object=GetAnimInstance(self=GetOwnerSkeletalMeshComponent()))
   > n13 = Switch on EFortWeaponCoreAnimation!(execute, Selection:EFortWeaponCoreAnimation Enum=Get WeaponCoreAnim(self=@n12)) -> (
    Melee:Exec
    Pistol:Exec
    Shotgun:Exec
    PaperBlueprint:Exec
    Rifle:Exec
    MeleeOneHand:Exec
    MachinePistol:Exec
    RocketLauncher:Exec
    GrenadeLauncher:Exec
    GoingCommando:Exec
    AssaultRifle:Exec
    TacticalShotgun:Exec
    SniperRifle:Exec
    TrapPlacement:Exec
    ShoulderLauncher:Exec
    AbilityDecoTool:Exec
    Crossbow:Exec
    C4:Exec
    RemoteControl:Exec
    DualWield:Exec
    AR_BullPup:Exec
    AR_ForwardGrip:Exec
    MedPackPaddles:Exec
    SMG_P90:Exec
    AR_DrumGun:Exec
    Consumable_Small:Exec
    Consumable_Large:Exec
    Balloon:Exec
    MountedTurret:Exec
    CreativeTool:Exec
    ExplosiveBow:Exec
    AshtonIndigo:Exec
    AshtonChicago:Exec
    MeleeDualWield:Exec
    Unarmed:Exec
   )
   Set 2H_range_weapon(
    execute=[
     @n13.AR_BullPup
     @n13.AR_DrumGun
     @n13.AR_ForwardGrip
     @n13.AbilityDecoTool
     @n13.AshtonChicago
     @n13.AshtonIndigo
     @n13.Balloon
     @n13.C4
     @n13.Consumable_Large
     @n13.Consumable_Small
     @n13.CreativeTool
     @n13.DualWield
     @n13.ExplosiveBow
     @n13.GrenadeLauncher
     @n13.MachinePistol
     @n13.MedPackPaddles
     @n13.Melee
     @n13.MeleeDualWield
     @n13.MeleeOneHand
     @n13.MountedTurret
     @n13.PaperBlueprint
     @n13.Pistol
     @n13.RemoteControl
     @n13.SMG_P90
     @n13.TrapPlacement
     @n13.Unarmed
    ]
    2H_range_weapon:bool=false
   ) -> (Output_Get:bool)
   Set 2H_range_weapon(
    execute=[
     @n13.AssaultRifle
     @n13.Crossbow
     @n13.GoingCommando
     @n13.Rifle
     @n13.RocketLauncher
     @n13.Shotgun
     @n13.ShoulderLauncher
     @n13.SniperRifle
     @n13.TacticalShotgun
    ]
    2H_range_weapon=true
   )
  ]}
]
variables [
 2H_range_weapon {type=boolean}
 CrouchingIdle {type=boolean}
 CrouchMoving {type=boolean}
 CrouchTargeting {type=boolean}
 CrouchTargeting 2H {type=boolean}
 Delta Time X {type=number}
 Gravity Override {type=Vector}
 Gravity Override Params {type=GravityOverrideParamsStruct}
 Head Offset {type=number}
 Head_rotation {type=number}
 Head_rotation_3 {type=number}
 Head_rotation_lowAlpha {type=number}
 jump_falling {type=boolean}
 NotTargeting {type=boolean}
 TEST {type=number}
 Weapon 2H {type=number}
]
